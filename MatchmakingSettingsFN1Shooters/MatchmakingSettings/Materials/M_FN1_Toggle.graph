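# M_FN1_Toggle is a translucent UI material that draws an on/off toggle switch for the matchmaking settings UI.
# IsOn serves as the progress ratio, with lerps for colors, a sliding offset, SDF knob corners and an expanding circle fill.
# Aspect-ratio-aware UVs and a rounded mask shape the knob and inner background.
# IsEnabled selects disabled colors, and the Ranked parameter collection value blends in an alternate selected background color.
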Material /MatchmakingSettingsFN1Shooters/MatchmakingSettings/Materials/M_FN1_Toggle
graph {
 properties {bAutomaticallySetUsageInEditor=false, blendMode=BLEND_Translucent, materialDomain=MD_UI, usedShadingModels=Default Lit}
 nodes [
  n1 = NamedRerouteUsage()
  LinearInterpolate(
   A=LinearInterpolate(B=LinearInterpolate(Alpha=@n1), Alpha=ScalarParameter() {defaultValue=1, parameterName=IsEnabled})
   Alpha=NamedRerouteUsage()
  ) {constA=0.15}
  n2 = Constant() {r=0.08}
  n3 = Max(A=NamedRerouteUsage()) {constB=0}
  n4 = MaterialFunctionCall(Size=OneMinus(None=Divide(A=@n2, B=@n3)), Corner Radius=Constant(), Radius in Pixels=Constant()) {materialFunction=&MF_UI_CoreUI_RoundedMask}
  LinearInterpolate(
   A=LinearInterpolate(
    A=Constant4Vector() {constant=<rgba 0.0048 0.004 0.0116 0.149>}.RGB
    B=NamedRerouteUsage()
    Alpha=NamedRerouteDeclaration(None=@n4.Shape Mask) {name=Inner Background Mask, nodeColor=<rgba 0 0 0 1>}
   )
   B=NamedRerouteUsage()
   Alpha=NamedRerouteUsage()
  )
  NamedRerouteDeclaration(None=ScalarParameter() {desc=Progress, parameterName=IsOn, sliderMax=1}) {name=Progress Ratio, nodeColor=<rgba 0 0 0 1>}
  n5 = VectorParameter() {parameterName=BackgroundColor_Idle, defaultValue=<rgba 0.0048 0.004 0.0116 1>}
  n6 = LinearInterpolate(
   A=VectorParameter() {parameterName=BackgroundColor_Selected, defaultValue=<rgba 0.9301 1 0.0103 1>}.RGB
   B=VectorParameter() {parameterName=BackgroundColor_Selected_Ranked, defaultValue=<rgba 0.5776 1 0 1>}.RGB
   Alpha=CollectionParameter() {collection=&/MPC_FN1_Matchmaking_Ranked, parameterName=Ranked}
  )
  n7 = NamedRerouteUsage()
  NamedRerouteDeclaration(None=LinearInterpolate(
    A=VectorParameter() {parameterName=Disabled_SelectedState, defaultValue=<rgba 0.0742 0.0395 0.3813 1>}.RGB
    B=LinearInterpolate(A=@n5.RGB, B=@n6, Alpha=@n7)
    Alpha=ScalarParameter() {defaultValue=1, parameterName=IsEnabled}
   )) {name=BackgroundColor, nodeColor=<rgba 0 0.2734 1 1>}
  n8 = NamedRerouteDeclaration(None=MaterialFunctionCall!(Dimensions) {materialFunction=/FortUILibrary/MaterialFunctions/Basics/MF_UI_GetAspectRatio}.Min Ratio) {name=Aspect Ratio, nodeColor=<rgba 0 0 0 1>}
  n9 = Multiply(A=Constant() {r=0.5}, B=OneMinus(None=ComponentMask(None=@n8) {g=true}))
  n10 = LinearInterpolate(A=@n9, B=Multiply(A=@n9) {constB=-1}, Alpha=NamedRerouteUsage()) {desc=Offset}
  n11 = Constant()
  n12 = MaterialFunctionCall(UV=Add(A=MaterialFunctionCall() {materialFunction=&GetUserInterfaceUV}.Normalized UV, B=AppendVector(A=@n10, B=@n11))) {materialFunction=&MF_UI_UVScaling}
  n13 = LinearInterpolate(A=Constant2Vector() {g=0.4, r=1}, B=Constant2Vector() {g=0.56, r=0.53}, Alpha=NamedRerouteUsage())
  n14 = LinearInterpolate(A=Constant2Vector() {g=-0.4, r=1}, B=Constant2Vector() {g=-0.56, r=0.53}, Alpha=NamedRerouteUsage())
  n15 = LinearInterpolate(A=Constant2Vector() {g=-0.56, r=-0.53}, B=Constant2Vector() {g=-0.4, r=-1}, Alpha=NamedRerouteUsage())
  n16 = LinearInterpolate(A=Constant2Vector() {g=0.56, r=-0.53}, B=Constant2Vector() {g=0.4, r=-1}, Alpha=NamedRerouteUsage())
  NamedRerouteDeclaration(None=MaterialFunctionCall(DistanceField=MaterialFunctionCall!(UVs=@n12.Scaled to Fit, Corner 01=@n13, Corner 02=@n14, Corner 03=@n15, Corner 04=@n16, Glow Min, Glow Max, Outline Thickness) {materialFunction=/FortUILibrary/MaterialFunctions/Shapes/MF_UI_WIP_Square}.SDF) {materialFunction=&MF_UI_AntiAliasing}.Inner) {name=Toggle Fill Mask, nodeColor=<rgba 0 0 0 1>}
  n17 = VectorParameter() {parameterName=Disabled_ToggleColor_Idle, defaultValue=<rgba 0.1912 0.1329 0.6038 1>}
  n18 = VectorParameter() {parameterName=ToggleColor_Idle, defaultValue=<rgba 1 1 1 0>}
  n19 = ScalarParameter() {defaultValue=1, parameterName=IsEnabled}
  n20 = VectorParameter() {parameterName=Disabled_ToggleColor_Selected, defaultValue=<rgba 0.0437 0.0242 0.2159 1>}
  n21 = VectorParameter() {parameterName=ToggleColor_Selected, defaultValue=<rgba 0.0048 0.004 0.0116 1>}
  n22 = ScalarParameter() {defaultValue=1, parameterName=IsEnabled}
  NamedRerouteDeclaration(None=LinearInterpolate(A=LinearInterpolate(A=@n17.RGB, B=@n18.RGB, Alpha=@n19), B=LinearInterpolate(A=@n20.RGB, B=@n21.RGB, Alpha=@n22), Alpha=NamedRerouteUsage())) {name=ToggleColor, nodeColor=<rgba 1 0.9141 0 1>}
  n23 = Constant() {r=0.625}
  n24 = NamedRerouteUsage()
  NamedRerouteDeclaration(None=MaterialFunctionCall(UVs=@n12.Scaled to Fit, Size=LinearInterpolate(A=@n23, Alpha=@n24) {constB=3}) {materialFunction=&MF_UI_Circle}.Fill) {name=Toggle Expand Fill, nodeColor=<rgba 0 0 0 1>}
 ]
}
