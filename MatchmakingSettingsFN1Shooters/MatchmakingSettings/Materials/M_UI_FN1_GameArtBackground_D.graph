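# Translucent UI material that draws a game art background from one Texture parameter.
# It layers the art scaled to fit over a copy scaled to fill and darkened, using step masks on the fitted UVs to choose between them.
# A linear gradient blends in BottomGradientColor.
# The Transition parameter fades the result in from FadeColor while easing the fitted art's scale from MaxArtScale to normal size.
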Material /MatchmakingSettingsFN1Shooters/MatchmakingSettings/Materials/M_UI_FN1_GameArtBackground_D
graph {
 properties {bAutomaticallySetUsageInEditor=false, blendMode=BLEND_Translucent, materialDomain=MD_UI, usedShadingModels=Default Lit}
 nodes [
  n1 = MaterialFunctionCall(TextureSize=TextureProperty(Tex=NamedRerouteUsage()), ScalePivotPoint=Constant2Vector() {g=0.5, r=1}) {materialFunction=&MF_UI_UVScaling}
  n2 = Multiply(A=TextureSample(UVs=@n1.Scaled To Fill, Tex=NamedRerouteUsage()) {samplerSource=SSM_Clamp_WorldGroupSettings}.RGB) {constB=0.2}
  n3 = Clamp(None=@n1.Scaled to Fit)
  n4 = ScalarParameter() {defaultValue=1.2, parameterName=MaxArtScale}
  n5 = Constant2Vector() {g=1, r=1}
  n6 = ScalarParameter() {defaultValue=1, parameterName=Transition}
  n7 = TextureSample(
   UVs=MaterialFunctionCall(UVs=@n3, Texture Scale=LinearInterpolate(A=@n4, B=@n5, Alpha=@n6)) {materialFunction=&ScaleUVsByCenter}.UVs
   Tex=NamedRerouteUsage()
  ) {samplerSource=SSM_Clamp_WorldGroupSettings}
  n8 = Max(A=Step(Y=@n3) {constX=0}, B=Step(X=@n3) {constY=1})
  n9 = OneMinus(None=Add(A=ComponentMask(None=@n8) {r=true}, B=ComponentMask(None=@n8) {g=true}))
  n10 = MaterialFunctionCall(GradientEnd=Constant() {r=0.5}, RotationAngle=Constant() {r=0.25}) {materialFunction=&MF_UI_LinearGradient}
  LinearInterpolate(
   A=VectorParameter() {parameterName=FadeColor, defaultValue=<rgba 0.0048 0.004 0.0116 1>}.RGB
   B=LinearInterpolate(
    A=VectorParameter() {parameterName=BottomGradientColor, defaultValue=<rgba 0.0048 0.004 0.0116 1>}.RGB
    B=LinearInterpolate(A=@n2, B=@n7.RGB, Alpha=@n9)
    Alpha=Power(Base=@n10.Result)
   )
   Alpha=ScalarParameter() {defaultValue=1, parameterName=Transition}
  )
  NamedRerouteDeclaration(None=TextureObjectParameter() {parameterName=Texture, samplerSource=SSM_Wrap_WorldGroupSettings}) {name=Texture, nodeColor=<rgba 0 1 0.6016 1>}
 ]
}
