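# Translucent UI material that tints an Icon texture by button state for the FN1 Shooters matchmaking settings UI.
# FortUILibrary modular button functions blend base and selectable states into State Color and State Opacity reroutes, alongside multiplies with the icon alpha.
# The selected idle color lerps between normal and "Raked" color parameters using the Ranked value from the MPC_FN1_Matchmaking_Ranked parameter collection.
# The base states' Color output feeds both the color and opacity inputs of the selectable function.
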Material /MatchmakingSettingsFN1Shooters/MatchmakingSettings/Materials/M_UI_FN1_Icon
graph {
 properties {blendMode=BLEND_Translucent, materialDomain=MD_UI, pixelDepthOffsetMode=PDOM_Legacy, usedShadingModels=Default Lit}
 nodes [
  n1 = TextureSample(Tex=TextureObjectParameter() {parameterName=Icon})
  Multiply(A=@n1.A, B=NamedRerouteUsage())
  Multiply(A=NamedRerouteUsage(), B=@n1.A)
  n2 = MaterialFunctionCall!(Idle, Idle Opacity, Hovered, Hovered Opacity, Focused, Focused Opacity, Pressed, Pressed Opacity, Disabled, Disabled Opacity) {materialFunction=/FortUILibrary/Systems/ModularButton/MaterialFunctions/MF_UIKit_ButtonStatesBase_Lerp}
  n3 = VectorParameter() {group=State colors, parameterName=Color_SelectedIdle, sortPriority=10, defaultValue=<rgba 1 1 1 1>}
  n4 = VectorParameter() {group=State colors, parameterName=Color_SelectedIdle_Raked, sortPriority=10, defaultValue=<rgba 0 0 0 1>}
  n5 = MaterialFunctionCall!(
   Base States=@n2.Color
   Base States Opacity=@n2.Color
   Selected_Idle=LinearInterpolate(
    A=@n3.RGB
    B=@n4.RGB
    Alpha=CollectionParameter() {collection=&/MPC_FN1_Matchmaking_Ranked, parameterName=Ranked}
   )
   Selected_Idle Opacity
   Selected_Hovered
   Selected_Hovered Opacity
   Selected_Focused
   Selected_Focused Opacity
   Selected_Pressed
   Selected_Pressed Opacity
   Selected_Disabled
   Selected_Disabled Opacity
  ) {materialFunction=/FortUILibrary/Systems/ModularButton/MaterialFunctions/MF_UIKit_ButtonStatesSelectable_Lerp}
  NamedRerouteDeclaration(None=@n5.Color) {name=State Color, nodeColor=<rgba 0 0 0 1>}
  NamedRerouteDeclaration(None=@n5.Opacity) {name=State Opacity, nodeColor=<rgba 0 0 0 1>}
 ]
}
