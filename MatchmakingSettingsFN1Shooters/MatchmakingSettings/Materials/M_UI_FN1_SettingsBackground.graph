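# Translucent UI material that draws the background for the FN1 Shooters matchmaking settings screen.
# It layers gradient-shaped color with linear-dodge blends, then lerps to another input using the Ranked material parameter collection value.
# Named reroutes organize radial gradients, an animated StormTech grid pattern and aspect-ratio-adjusted UVs.
# Rotated-UV smoothsteps form an angled shape mask and, by subtracting an offset copy, an edge line, with IsFullBg? switching the mask mode.
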Material /MatchmakingSettingsFN1Shooters/MatchmakingSettings/Materials/M_UI_FN1_SettingsBackground
graph {
 properties {bAutomaticallySetUsageInEditor=false, blendMode=BLEND_Translucent, materialDomain=MD_UI, usedShadingModels=Default Lit}
 nodes [
  n1 = Saturate(None=MaterialFunctionCall(GradientStart=Constant(), GradientEnd=Constant() {r=0.2}, RotationAngle=Constant() {r=0.28}, GradientOffset=Constant() {r=-0.15}) {materialFunction=&MF_UI_LinearGradient}.Result)
  n2 = LinearInterpolate(A=NamedRerouteUsage(), B=NamedRerouteUsage(), Alpha=NamedRerouteUsage())
  n3 = NamedRerouteUsage()
  n4 = MaterialFunctionCall!(Blend=LinearInterpolate(A=Multiply(A=@n1, B=@n2), B=NamedRerouteUsage(), Alpha=Multiply(A=@n3) {constB=0.15}), Base=NamedRerouteUsage()) {materialFunction=/Engine/Functions/Engine_MaterialFunctions03/Blends/Blend_LinearDodge}
  LinearInterpolate(
   A=LinearInterpolate(
    A=@n4.Result
    B=MaterialFunctionCall(Blend=@n4.Result, Base=@n4.Result) {materialFunction=&Blend_LinearDodge}.Result
    Alpha=NamedRerouteUsage()
   )
   B=NamedRerouteUsage()
   Alpha=CollectionParameter() {collection=&/MPC_FN1_Matchmaking_Ranked, parameterName=Ranked}
  )
  n5 = MaterialFunctionCall(UVs=NamedRerouteUsage(), CenterPosition=Constant2Vector() {g=1.2119, r=0.2048}, Radius=Constant() {r=0.6928}, Density=Constant() {r=1.1386}) {materialFunction=&RadialGradientExponential}
  n6 = Constant4Vector() {constant=<rgba 0.1946 0.006 1 1>}
  n7 = MaterialFunctionCall(UVs=NamedRerouteUsage(), CenterPosition=Constant2Vector() {g=0.5835, r=0.0638}, Radius=Constant() {r=0.6352}, Density=Constant() {r=1.1386}) {materialFunction=&RadialGradientExponential}
  NamedRerouteDeclaration(None=Add(A=Multiply(A=@n5.RadialGradientExponential, B=@n6.RGB), B=Multiply(A=@n7.RadialGradientExponential, B=@n6.RGB))) {name=GradientColor_2, nodeColor=<rgba 1 0.7031 0 1>}
  n8 = MaterialFunctionCall(CenterPosition=Constant2Vector() {g=-0.2496, r=1.2882}, Radius=Constant() {r=1.3185}, Density=Constant() {r=1.1642}) {materialFunction=&RadialGradientExponential}
  NamedRerouteDeclaration(None=@n8.RadialGradientExponential) {name=RadialGradient, nodeColor=<rgba 0.7578 0 1 1>}
  NamedRerouteDeclaration(None=LinearInterpolate(A=Constant4Vector() {constant=<rgba 0 0 0.0262 1>}.RGB, B=Constant4Vector() {constant=<rgba 0 0 0.0976 1>}.RGB, Alpha=NamedRerouteUsage())) {name=RankedColor, nodeColor=<rgba 0 1 0.4375 1>}
  n9 = Constant() {r=0.2}
  n10 = Constant() {r=2}
  n11 = ScalarParameter() {defaultValue=-0.03, parameterName=Angle}
  NamedRerouteDeclaration(None=Divide(A=MaterialFunctionCall(GradientStart=@n9, GradientEnd=@n10, RotationAngle=@n11) {materialFunction=&MF_UI_LinearGradient}.Result) {constB=4}) {name=PatternOpacity, nodeColor=<rgba 0.2656 0 1 1>}
  n12 = NamedRerouteUsage()
  n13 = If(A=NamedRerouteUsage(), "A > B"=@n12, "A == B"=@n12, "A < B"=TextureCoordinate()) {constB=0.85}
  n14 = MaterialFunctionCall(UVs=@n13, "Rotation Angle (0-1)"=ScalarParameter() {defaultValue=-0.03, parameterName=Angle}) {materialFunction=&CustomRotator}
  n15 = SmoothStep() {constMax=0, constMin=1}
  n16 = ScalarParameter() {parameterName=IsFullBg?}
  n17 = LinearInterpolate(A=SmoothStep(Value=ComponentMask(None=@n14.Rotated Values) {r=true}) {constMax=0.1015, constMin=0.1}, B=@n15, Alpha=@n16) {constAlpha=1}
  NamedRerouteDeclaration(None=@n17) {name=Mask, nodeColor=<rgba 0.9688 0 1 1>}
  NamedRerouteDeclaration(None=Constant4Vector() {constant=<rgba 0.0048 0.004 0.0116 1>}.RGB) {name=BaseShapeColor, nodeColor=<rgba 0 0.9766 1 1>}
  NamedRerouteDeclaration(None=Multiply(A=@n8.RadialGradientExponential, B=Constant4Vector() {constant=<rgba 0.0578 0.0097 1 1>}.RGB)) {name=GradientColor_1, nodeColor=<rgba 1 0.7031 0 1>}
  n18 = TextureCoordinate()
  n19 = ComponentMask(None=@n18) {r=true}
  n20 = NamedRerouteUsage()
  NamedRerouteDeclaration(None=AppendVector(A=Multiply(A=@n19, B=@n20), B=ComponentMask(None=@n18) {g=true})) {name=AspectRatio_Left, nodeColor=<rgba 0.3828 0 1 1>}
  NamedRerouteDeclaration(None=Max(A=@n5.RadialGradientExponential, B=Multiply(A=@n7.RadialGradientExponential))) {name=GradientBase_2, nodeColor=<rgba 1 0.5156 0 1>}
  n21 = MaterialFunctionCall(
   UVs=Add(A=@n13, B=Constant2Vector() {g=-0.09})
   "Rotation Angle (0-1)"=Subtract(A=ScalarParameter() {defaultValue=-0.03, parameterName=Angle}) {constB=0.004}
  ) {materialFunction=&CustomRotator}
  n22 = LinearInterpolate(A=SmoothStep(Value=ComponentMask(None=@n21.Rotated Values) {r=true}) {constMax=0.1015, constMin=0.1}, B=@n15, Alpha=@n16) {constAlpha=1}
  NamedRerouteDeclaration(None=Multiply(A=Subtract(A=@n17, B=@n22)) {constB=0.2}) {name=LineMask, nodeColor=<rgba 1 0.7266 0 1>}
  n23 = ScalarParameter() {defaultValue=15, group=Pattern, parameterName=Tiling}
  n24 = Constant() {r=-0.03}
  n25 = Multiply(A=Time(), B=ScalarParameter() {defaultValue=0.25, group=Pattern, parameterName=Grid Speed})
  n26 = Constant() {r=0.05}
  n27 = ScalarParameter() {group=Pattern, parameterName=Main Reduction}
  n28 = Constant2Vector() {g=0.3}
  n29 = Constant() {r=0.025}
  n30 = Constant2Vector() {g=0.5, r=0.75}
  n31 = Constant()
  n32 = ScalarParameter() {defaultValue=1, group=Pattern, parameterName=Sub Reduction}
  n33 = Constant2Vector() {g=0.2, r=0.1}
  n34 = Constant2Vector() {g=0.85, r=0.925}
  n35 = Constant()
  NamedRerouteDeclaration(None=LinearInterpolate(
    A=VectorParameter() {group=Backplate, parameterName=Backplate Color, defaultValue=<rgba 0.0204 0.0036 0.0906 1>}.RGB
    B=VectorParameter() {group=Pattern, parameterName=Grid Color, defaultValue=<rgba 0.2159 0 1 1>}.RGB
    Alpha=MaterialFunctionCall!(
     Screen UV
     Tiling=@n23
     Rotation=@n24
     Main Min/Max Opacity=@n33
     Main Min/Max Scale=@n34
     Main Min/Max Perspective=@n35
     Main Outline Thickness
     Main Speed=@n26
     Main Reduction=@n27
     Sub Min/Max Opacity=@n28
     Sub Min/Max Scale=@n30
     Sub Min/Max Perspective=@n31
     Sub Outline Thickness
     Sub Speed=@n29
     Sub Reduction=@n32
     Random Seed=@n25
     Time
    ) {materialFunction=/FortUILibrary/MaterialFunctions/BR/MF_UI_BR_StormTech}.Grid
   )) {name=Pattern, nodeColor=<rgba 1 0.5859 0 1>}
  n36 = MaterialFunctionCall() {materialFunction=&GetUserInterfaceUV}
  NamedRerouteDeclaration(None=Divide(A=ComponentMask(None=@n36.Pixel Size) {r=true}, B=ComponentMask(None=@n36.Pixel Size) {g=true})) {name=AspectRatio, nodeColor=<rgba 0 0.0859 1 1>}
  NamedRerouteUsage()
 ]
}
