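# Material function, exposed to the library under Tech Art as "Get Texture Resolution", that returns the pixel width and height of an input Texture2D.
# The texture object feeds a Custom HLSL node that calls GetDimensions on mip 0.
# The node outputs width and height as a float2 through the Resolution output, discarding the mip count.
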MaterialFunction /Game/Global/MaterialFunctions/MF_GetTextureResolution
graph {
 properties {bExposeToLibrary=true, userExposedCaption=Get Texture Resolution, libraryCategoriesText [Tech Art]}
 nodes [
  MaterialExpressionFunctionInput_0 = FunctionInput(Preview=TextureObject()) {inputName=TextureObject, inputType=FunctionInput_Texture2D}
  n1 = Custom!(Tex=@MaterialExpressionFunctionInput_0) {
   code="uint3 res;\r\nTex.GetDimensions(0,res.x,res.y,res.z);\r\nreturn float2(res.xy);"
   description=GetTextureResolution
   outputType=CMOT_Float2
   inputs [{inputName=Tex, input {expression=&self:MaterialExpressionFunctionInput_0}}]
  }
  FunctionOutput(None=@n1) {outputName=Resolution}
 ]
}
