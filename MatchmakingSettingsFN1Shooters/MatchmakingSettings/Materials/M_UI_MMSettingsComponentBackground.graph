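# Translucent UI material that draws the background of matchmaking settings components across button states.
# Idle, disabled and selected looks feed the shared modular button Lerp functions for base and selectable states.
# Rotated UVs with SmoothSteps cut slanted texture masks, lines and selection underlines; IdleTransition shifts the focused mask, and aspect-corrected UVs support fitting.
# Parameters control colors and texture use, a collection value switches Ranked colors, and an MSDF supplies the Blitz pattern mask.
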Material /MatchmakingSettingsFN1Shooters/MatchmakingSettings/Materials/M_UI_MMSettingsComponentBackground
graph {
 properties {bAutomaticallySetUsageInEditor=false, blendMode=BLEND_Translucent, materialDomain=MD_UI, usedShadingModels=Default Lit}
 nodes [
  n1 = MaterialFunctionCall!(
   Idle=NamedRerouteUsage()
   Idle Opacity=NamedRerouteUsage()
   Hovered
   Hovered Opacity
   Focused
   Focused Opacity
   Pressed
   Pressed Opacity
   Disabled
   Disabled Opacity
  ) {materialFunction=/FortUILibrary/Systems/ModularButton/MaterialFunctions/MF_UIKit_ButtonStatesBase_Lerp}
  n2 = NamedRerouteUsage()
  MaterialFunctionCall!(
   Base States=@n1.Color
   Base States Opacity=@n1.Color
   Selected_Idle=NamedRerouteUsage()
   Selected_Idle Opacity=NamedRerouteUsage()
   Selected_Hovered=@n2
   Selected_Hovered Opacity
   Selected_Focused=@n2
   Selected_Focused Opacity
   Selected_Pressed=@n2
   Selected_Pressed Opacity
   Selected_Disabled
   Selected_Disabled Opacity
  ) {materialFunction=/FortUILibrary/Systems/ModularButton/MaterialFunctions/MF_UIKit_ButtonStatesSelectable_Lerp}
  n3 = TextureCoordinate()
  n4 = MaterialFunctionCall() {materialFunction=&GetUserInterfaceUV}
  n5 = Multiply(
   A=OneMinus(None=ComponentMask(None=@n3) {r=true})
   B=Divide(A=ComponentMask(None=@n4.Pixel Size) {r=true}, B=ComponentMask(None=@n4.Pixel Size) {g=true})
  )
  NamedRerouteDeclaration(None=AppendVector(A=OneMinus(None=@n5), B=ComponentMask(None=@n3) {g=true})) {name=AspectRatio_Right, nodeColor=<rgba 0.3828 0 1 1>}
  n6 = NamedRerouteUsage()
  n7 = NamedRerouteUsage()
  n8 = Multiply(A=@n7, B=NamedRerouteUsage())
  n9 = NamedRerouteUsage()
  n10 = LinearInterpolate(
   A=NamedRerouteUsage()
   B=LinearInterpolate(A=LinearInterpolate(A=@n6, B=@n8, Alpha=@n7), Alpha=Multiply(A=@n9) {constB=0.05})
   Alpha=NamedRerouteUsage()
  )
  n11 = LinearInterpolate(A=NamedRerouteUsage(), Alpha=NamedRerouteUsage()) {constB=0}
  n12 = NamedRerouteUsage()
  NamedRerouteDeclaration(None=LinearInterpolate(A=@n10, B=LinearInterpolate(A=@n10, B=@n11, Alpha=@n12), Alpha=NamedRerouteUsage())) {name=Idle, nodeColor=<rgba 0.0176 0.004 0.0545 1>}
  NamedRerouteDeclaration(None=ScalarParameter() {parameterName=UseTexture}) {name=UseTexture, nodeColor=<rgba 1 0 0.9141 1>}
  n13 = VectorParameter() {desc=Please use this param to set selected color, parameterName=BaseColor, defaultValue=<rgba 0.0137 0.0015 0.0513 0.3>}
  NamedRerouteDeclaration(None=@n13.A) {name=IdleOpacity, nodeColor=<rgba 0 1 0.7656 1>}
  n14 = MaterialFunctionCall(
   UVs=NamedRerouteUsage()
   Rotation Center=Constant2Vector()
   "Rotation Angle (0-1)"=ScalarParameter() {defaultValue=0.0448, parameterName=Angle}
  ) {materialFunction=&CustomRotator}
  n15 = SmoothStep(Value=Add(A=ComponentMask(None=@n14.Rotated Values) {r=true}) {constB=0.5904}) {constMax=0.004}
  n16 = Add(A=NamedRerouteUsage(), B=Constant2Vector())
  n17 = Subtract(A=ScalarParameter() {defaultValue=0.0448, parameterName=Angle}) {constB=0.006}
  n18 = Add(A=ComponentMask(None=MaterialFunctionCall(UVs=@n16, "Rotation Angle (0-1)"=@n17) {materialFunction=&CustomRotator}.Rotated Values) {r=true}) {constB=0.3792}
  NamedRerouteDeclaration(None=Subtract(A=@n15, B=SmoothStep(Value=@n18) {constMax=0.004})) {name=TextureLine, nodeColor=<rgba 1 0.4453 0 1>}
  n19 = MaterialFunctionCall(
   TextureSize=MaterialFunctionCall!(TextureObject=NamedRerouteUsage()) {materialFunction=/Game/Global/MaterialFunctions/MF_GetTextureResolution}.Resolution
   ScalePivotPoint=Constant2Vector() {r=1}
  ) {materialFunction=&MF_UI_UVScaling}
  n20 = Constant2Vector() {g=1.2, r=1.2}
  NamedRerouteDeclaration(None=TextureSample(
    UVs=MaterialFunctionCall(UVs=@n19.Scaled to Fit, Texture Scale=@n20) {materialFunction=&ScaleUVsByCenter}.UVs
    Tex=NamedRerouteUsage()
   ) {samplerSource=SSM_Clamp_WorldGroupSettings}.RGB) {name=Texture_Scaled, nodeColor=<rgba 0 0.2734 1 1>}
  n21 = ScalarParameter() {defaultValue=0.0448, parameterName=Angle}
  n22 = LinearInterpolate(Alpha=ScalarParameter() {parameterName=IdleTransition}) {constB=0.0182}
  n23 = ComponentMask(None=MaterialFunctionCall(UVs=NamedRerouteUsage(), Rotation Center=Constant2Vector(), "Rotation Angle (0-1)"=Subtract(A=@n21, B=@n22) {constB=0.0182}) {materialFunction=&CustomRotator}.Rotated Values) {r=true}
  n24 = LinearInterpolate(Alpha=ScalarParameter() {parameterName=IdleTransition}) {constA=0.5904, constB=0.776}
  NamedRerouteDeclaration(None=SmoothStep(Value=Add(A=@n23, B=@n24) {constB=0.776}) {constMax=0.004}) {name=FocusedTextureOpacity, nodeColor=<rgba 1 0 0.3281 1>}
  NamedRerouteDeclaration(None=ScalarParameter() {parameterName=HorizontalAngle}) {name=HorizontalAngle, nodeColor=<rgba 0.5938 0 1 1>}
  NamedRerouteDeclaration(None=ScalarParameter() {parameterName=IsPureChaos}) {name=IsPureChaos, nodeColor=<rgba 1 0 0.3984 1>}
  n25 = TextureSample(UVs=MaterialFunctionCall(Texture Scale=Constant() {r=1.01}) {materialFunction=&ScaleUVsByCenter}.UVs) {samplerType=SAMPLERTYPE_LinearColor, texture=/MatchmakingSettingsFN1Shooters/MatchmakingSettings/Textures/T_UI_Icon_BlitzPattern_MSDF}
  NamedRerouteDeclaration(None=MaterialFunctionCall(
    DistanceField=MaterialFunctionCall!(MSDF RGB=@n25.RGB) {materialFunction=/FortUILibrary/MaterialFunctions/Shapes/Modifiers/MF_UI_MSDFtoSDF}.SDF
    Threshold=Constant() {r=0.4616}
    Smoothing=Constant() {r=0.0116}
   ) {materialFunction=&MF_UI_AntiAliasing}.Outer) {name=BlitzVisualsMask, nodeColor=<rgba 1 0.7969 0 1>}
  n26 = Constant4Vector() {constant=<rgba 0.0742 0.0395 0.3813 1>}
  n27 = NamedRerouteUsage()
  n28 = VectorParameter() {parameterName=DisabledColor, defaultValue=<rgba 0.1022 0.0595 0.4508 1>}
  n29 = NamedRerouteUsage()
  NamedRerouteDeclaration(None=LinearInterpolate(A=@n28.RGB, B=LinearInterpolate(A=@n28.RGB, B=@n26.RGB, Alpha=@n27), Alpha=@n29)) {name=DisabledColor, nodeColor=<rgba 0 1 0.0391 1>}
  n30 = NamedRerouteUsage()
  NamedRerouteDeclaration(None=LinearInterpolate(A=@n28.A, B=LinearInterpolate(A=@n28.A, Alpha=@n30), Alpha=@n29)) {name=DisabledOpacity, nodeColor=<rgba 1 0 0.2812 1>}
  n31 = NamedRerouteUsage()
  NamedRerouteDeclaration(None=NamedRerouteDeclaration(None=Subtract(A=@n15, B=@n31)) {name=IdleTextureOpacity, nodeColor=<rgba 1 0 0.3281 1>}) {name=TextureOpacity, nodeColor=<rgba 0 0.9766 1 1>}
  n32 = NamedRerouteUsage()
  n33 = Constant2Vector() {g=0.06}
  n34 = SmoothStep(Value=ComponentMask(None=Add(A=@n32, B=@n33)) {g=true}) {constMax=0.98, constMin=0.96}
  NamedRerouteDeclaration(None=@n34) {name=Lineshape, nodeColor=<rgba 0.3828 1 0 1>}
  NamedRerouteDeclaration(None=LinearInterpolate(
    A=VectorParameter() {parameterName=SelectedUnderlineColor, defaultValue=<rgba 0.9301 1 0.0091 1>}.RGB
    B=VectorParameter() {parameterName=SelectedUnderlineColor_Ranked, defaultValue=<rgba 1 1 1 1>}.RGB
    Alpha=NamedRerouteUsage()
   )) {name=SelectedUnderline_Color, nodeColor=<rgba 1 0 0.6328 1>}
  n35 = VectorParameter() {desc=Please use this param to set selected color, parameterName=SelectedColor, defaultValue=<rgba 0 0.0908 0.9911 1>}
  n36 = VectorParameter() {desc=Please use this param to set selected color, parameterName=SelectedColor_Ranked, defaultValue=<rgba 0.5776 1 0 1>}
  n37 = Add(A=NamedRerouteUsage(), B=Constant2Vector())
  n38 = LinearInterpolate(Alpha=SmoothStep(Value=ComponentMask(None=@n37) {g=true}) {constMax=0.98, constMin=0.96})
  n39 = LinearInterpolate(A=@n35.RGBA, B=LinearInterpolate(A=@n36.RGBA, B=@n38, Alpha=@n34) {constA=1}, Alpha=NamedRerouteUsage())
  NamedRerouteDeclaration(None=ComponentMask(None=@n39) {b=true, g=true, r=true}) {name=SelectedBaseColor, nodeColor=<rgba 0 0.6719 1 1>}
  NamedRerouteDeclaration(None=CollectionParameter() {collection=&/MPC_FN1_Matchmaking_Ranked, parameterName=Ranked}) {name=Ranked, nodeColor=<rgba 1 0.1172 0 1>}
  n40 = LinearInterpolate(
   A=VectorParameter() {parameterName=SelectedUnderlineColor, defaultValue=<rgba 0.9301 1 0.0091 1>}.RGBA
   B=VectorParameter() {parameterName=SelectedUnderlineColorFocused_Ranked, defaultValue=<rgba 0.5776 1 0 1>}.RGBA
   Alpha=NamedRerouteUsage()
  )
  NamedRerouteDeclaration(None=ComponentMask(None=@n40) {b=true, g=true, r=true}) {name=FocusedSelectedUnderlineColor, nodeColor=<rgba 0 0.2031 1 1>}
  NamedRerouteDeclaration(None=@n38) {name=UndelineShape, nodeColor=<rgba 1 0 0.7031 1>}
  NamedRerouteDeclaration(None=@n13.RGB) {name=BaseColor, nodeColor=<rgba 0.8047 0 1 1>}
  NamedRerouteDeclaration(None=Constant4Vector() {constant=<rgba 0.8714 0.3712 0.0212 1>}.RGB) {name=BlitzPatternColor, nodeColor=<rgba 0 1 0.4844 1>}
  n41 = NamedRerouteUsage()
  n42 = LinearInterpolate(A=NamedRerouteUsage(), B=NamedRerouteUsage(), Alpha=SmoothStep(Value=ComponentMask(None=@n41) {g=true}) {constMax=0.98, constMin=0.97})
  n43 = LinearInterpolate(A=@n42, B=NamedRerouteUsage(), Alpha=NamedRerouteUsage())
  n44 = NamedRerouteUsage()
  n45 = ComponentMask(None=NamedRerouteUsage()) {b=true, g=true, r=true}
  n46 = NamedRerouteUsage()
  n47 = NamedRerouteUsage()
  NamedRerouteDeclaration(None=LinearInterpolate(A=LinearInterpolate(A=@n42, B=@n43, Alpha=@n44), B=LinearInterpolate(A=@n45, B=@n46, Alpha=@n47), Alpha=NamedRerouteUsage())) {name=SelectedColor, nodeColor=<rgba 0 0.1604 1 1>}
  NamedRerouteDeclaration(None=ComponentMask(None=@n39) {a=true}) {name=SelectedAlpha, nodeColor=<rgba 0 0.1562 1 1>}
  NamedRerouteDeclaration(None=ComponentMask(None=@n40) {a=true}) {name=FocusedSelectedUnderlineColorAlpha, nodeColor=<rgba 0 0.1562 1 1>}
  NamedRerouteDeclaration(None=TextureObject()) {name=TextureObj, nodeColor=<rgba 0.3125 0 1 1>}
  n48 = NamedRerouteUsage()
  n49 = NamedRerouteUsage()
  NamedRerouteDeclaration(None=LinearInterpolate(A=@n48, B=LinearInterpolate(A=@n48, Alpha=@n49), Alpha=NamedRerouteUsage())) {name=SelectedOpacity, nodeColor=<rgba 0 0.1604 1 1>}
  n50 = LinearInterpolate(B=LinearInterpolate(B=NamedRerouteUsage(), Alpha=NamedRerouteUsage()), Alpha=NamedRerouteUsage()) {constA=1}
  n51 = NamedRerouteUsage()
  n52 = NamedRerouteUsage()
  NamedRerouteDeclaration(None=LinearInterpolate(A=@n50, B=LinearInterpolate(A=@n50, B=@n51, Alpha=@n52), Alpha=NamedRerouteUsage())) {name=FocusedSelected, nodeColor=<rgba 0.3359 0 1 1>}
  n53 = MaterialFunctionCall() {materialFunction=&GetUserInterfaceUV}
  n54 = ComponentMask(None=@n53.Normalized UV) {r=true}
  n55 = OneMinus(None=ComponentMask(None=@n53.Normalized UV) {g=true})
  n56 = Divide(A=ComponentMask(None=@n53.Pixel Size) {g=true}, B=ComponentMask(None=@n53.Pixel Size) {r=true})
  n57 = OneMinus(None=Max(A=Multiply(A=@n55, B=@n56), B=@n55))
  n58 = Constant() {r=0.01}
  n59 = Clamp(None=@n56) {minDefault=0.7}
  NamedRerouteDeclaration(None=MaterialFunctionCall(UVs=AppendVector(A=@n54, B=@n57), Rotation Center=Constant2Vector() {g=1}, "Rotation Angle (0-1)"=Divide(A=@n58, B=@n59)) {materialFunction=&CustomRotator}.Rotated Values) {name=SegmentUV, nodeColor=<rgba 0 1 0.0156 1>}
  NamedRerouteDeclaration(None=TextureSample(UVs=@n19.Scaled to Fit, Tex=NamedRerouteUsage()) {samplerSource=SSM_Clamp_WorldGroupSettings}.RGB) {name=Texture, nodeColor=<rgba 0 0.2734 1 1>}
 ]
}
