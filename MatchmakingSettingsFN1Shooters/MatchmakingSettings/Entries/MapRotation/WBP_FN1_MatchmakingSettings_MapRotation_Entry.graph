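# Map rotation entry widget for FN1 Shooters matchmaking settings, stacking current and next map widgets vertically.
# An "Incoming:" label and countdown text sit in the next map widget's named slot.
# On Construct it binds to visibility changes, each of which refreshes SelectedIslandViewModel from the MatchmakingUISubsystem.
# HandleNextMapRotationDateSet, not called in these graphs, sets up and starts the countdown timer view model with the given date.
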WidgetBlueprint /MatchmakingSettingsFN1Shooters/MatchmakingSettings/Entries/MapRotation/WBP_FN1_MatchmakingSettings_MapRotation_Entry
graphs [
 EventGraph {nodes [
   n1 = CustomEvent!() -> (OutputDelegate, then, InVisibility:ESlateVisibility Enum=Visible) {is=title K2Node_CustomEvent}
   > SetVM!(
    execute
    self
    SelectedIslandViewModel:Selected Island View Model Object Reference=GetSelectedIslandViewModel!(self:Matchmaking UISubsystem Object Reference=Get MatchmakingUISubsystem() -> (ReturnValue:Matchmaking UISubsystem Object Reference) {is=title K2Node_GetSubsystem}) -> (ReturnValue:Selected Island View Model Object Reference)
   ) -> (then)
   Event Construct()
   > Bind Event to On Visibility Changed(self:User Widget Object Reference, Delegate=@n1) {is=title K2Node_AddDelegate}
  ]}
 HandleNextMapRotationDateSet {nodes [
   n1 = HandleNextMapRotationDateSet!() -> (then, DateTime:Date Time Structure) {is=title K2Node_FunctionEntry}
   n2 = Get FortCountdownTimerVM() -> (FortCountdownTimerVM:Countdown Timer View Model Object Reference)
   > SetupViewModel!(
    execute
    self:Countdown Timer View Model Object Reference=@n2
    InFutureTime:"Date Time Structure (by ref)"=@n1
    InLiveText:"Text (by ref)"=GetEmptyText!() -> (ReturnValue:Text)
   ) -> (then)
   > StartTimer(self:Countdown Timer View Model Object Reference=@n2)
  ]}
 SetVM {nodes [
   n1 = SetVM!() -> (then, SelectedIslandViewModel:Selected Island View Model Object Reference) {is=title K2Node_FunctionEntry}
   > Set SelectedIslandViewModel(SelectedIslandViewModel:Selected Island View Model Object Reference=@n1) -> (Output_Get:Selected Island View Model Object Reference)
  ]}
]
variables []
widgets {
 info {inheritedWidgetCount=0, namedSlotCount=1, parentClass=&UserWidget, rootWidgetClass=&StackBox, widgetCount=9}
 widgets [
  HB_Maps {widgetClassPath=&StackBox, widget=HB_Maps, properties {bOverride_Cursor=true, orientation=Orient_Vertical}}
  MapRotationWidgetCurrent {
   widgetClassPath=&/WBP_FN1_MatchmakingSettings_MapRotationWidget
   bIsVariable=true
   parent=HB_Maps
   widget=MapRotationWidgetCurrent
   slot {class=StackBoxSlot}
  }
  MapRotationWidgetNext {
   widgetClassPath=&WBP_FN1_MatchmakingSettings_MapRotationWidget
   bIsVariable=true
   parent=HB_Maps
   widget=MapRotationWidgetNext
   properties {isCurrentMap=false}
   slot {class=StackBoxSlot}
  }
  O_Timer {widgetClassPath=&Overlay, namedSlotHost=MapRotationWidgetNext, widget=O_Timer}
  VerticalBox_0 {
   widgetClassPath=&VerticalBox
   parent=O_Timer
   widget=VerticalBox_0
   slot {class=OverlaySlot, properties {padding=<ltrb 24 16.5 0 0>}}
  }
  Incoming {
   widgetClassPath=&CommonTextBlock
   parent=VerticalBox_0
   widget=Incoming
   properties {
    autoWrapText=true
    minDesiredWidth=260
    scrollStyle=&TextScrollStyle-Base_C
    text=Incoming:
    colorAndOpacity {colorUseRule=UseColor_Specified, specifiedColor=<rgba 1 1 1 0.5>}
    font {
     bForceMonospaced=false
     bMaterialIsStencil=false
     fontMaterial=null
     fontObject=&HeadingNow
     letterSpacing=0
     monospacedWidth=1
     size=15
     skewAmount=0
     typefaceFontName=75 Medium
     outlineSettings {
      bApplyOutlineToDropShadows=false
      bMiteredCorners=false
      bSeparateFillAlpha=false
      outlineMaterial=null
      outlineSize=0
      outlineColor=<rgba 0 0 0 1>
     }
    }
    shadowOffset=<0 0>
   }
   slot {class=VerticalBoxSlot}
  }
  HorizontalBox_0 {widgetClassPath=&HorizontalBox, parent=VerticalBox_0, widget=HorizontalBox_0, slot {class=VerticalBoxSlot}}
  Image_40 {
   widgetClassPath=&Image
   bIsVariable=true
   parent=HorizontalBox_0
   widget=Image_40
   properties {brush {
     bIsDynamicallyLoaded=false
     drawAs=Image
     imageType=FullColor
     mirroring=NoMirror
     resourceName=None
     resourceObject=/FortUILibrary/Textures/MaterialInstances/MI_UIKit_Icon_UpdatePending
     tiling=NoTile
     imageSize=<20 20>
     margin=<ltrb 0 0 0 0>
     outlineSettings {
      bUseBrushTransparency=false
      roundingType=HalfHeightRadius
      width=0
      color {colorUseRule=UseColor_Specified, specifiedColor=<rgba 0 0 0 0>}
      cornerRadii=<0 0 0 0>
     }
     tintColor {colorUseRule=UseColor_Specified, specifiedColor=<rgba 1 1 1 1>}
     uVRegion {bIsValid=false, max=<0 0>, min=<0 0>}
    }}
   slot {class=HorizontalBoxSlot, properties {verticalAlignment=VAlign_Center, padding=<ltrb 0 0 8 0>}}
  }
  Timer {
   widgetClassPath=&CommonTextBlock
   bIsVariable=true
   parent=HorizontalBox_0
   widget=Timer
   properties {
    autoWrapText=true
    minDesiredWidth=260
    scrollStyle=&TextScrollStyle-Base_C
    text=$Time
    font {
     bForceMonospaced=false
     bMaterialIsStencil=false
     fontMaterial=null
     fontObject=&HeadingNow
     letterSpacing=0
     monospacedWidth=1
     size=18
     skewAmount=0
     typefaceFontName=86 Bold
     outlineSettings {
      bApplyOutlineToDropShadows=false
      bMiteredCorners=false
      bSeparateFillAlpha=false
      outlineMaterial=null
      outlineSize=0
      outlineColor=<rgba 0 0 0 1>
     }
    }
    shadowOffset=<0 0>
   }
   slot {class=HorizontalBoxSlot}
  }
 ]
}
