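# UI translucent material that draws the FN1 ranked pattern from an MSDF texture, mirrored vertically.
# V is remapped around the midline and inverted on one half through an If node, folding the texture into two reflected halves.
# The UVs get a slight horizontal scale about the center before sampling the MSDF parameter.
# MF_UI_MSDFtoSDF converts the sample to an SDF, and a SmoothStep turns it into a soft-edged mask.
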Material /MatchmakingSettingsFN1Shooters/MatchmakingSettings/Materials/M_UI_FN1_RankedPattern
graph {
 properties {bAutomaticallySetUsageInEditor=false, blendMode=BLEND_Translucent, materialDomain=MD_UI, usedShadingModels=Default Lit}
 nodes [
  n1 = TextureCoordinate()
  n2 = ComponentMask(None=@n1) {r=true}
  n3 = Multiply(A=Subtract(A=ComponentMask(None=@n1) {g=true}) {constB=0.5}) {constB=2}
  n4 = If(A=@n3, "A > B"=OneMinus(None=@n3), "A == B"=@n3, "A < B"=@n3)
  n5 = TextureSampleParameter2D(UVs=MaterialFunctionCall(UVs=AppendVector(A=@n2, B=@n4), Texture Scale=Constant2Vector() {g=1, r=1.05}) {materialFunction=&ScaleUVsByCenter}.UVs) {
   group=2 - Distance Field
   parameterName=MSDF
   samplerType=SAMPLERTYPE_LinearColor
   texture=/MatchmakingSettingsFN1Shooters/MatchmakingSettings/Textures/T_UI_Icon_FN1_RankedPattern
  }
  SmoothStep(Value=MaterialFunctionCall!(MSDF RGB=@n5.RGB) {materialFunction=/FortUILibrary/MaterialFunctions/Shapes/Modifiers/MF_UI_MSDFtoSDF}.SDF) {constMax=0.9, constMin=0.6}
 ]
}
