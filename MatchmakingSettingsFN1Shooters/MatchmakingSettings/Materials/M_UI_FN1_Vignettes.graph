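# Translucent UI material that builds a dark, near-black blue vignette, apparently for the FN1 Shooters matchmaking settings screens.
# An exponential radial gradient on X-scaled UI UVs is inverted into an edge mask, then multiplied by the tint color.
# An If node passes the mask only when its input, apparently the pixel-size width/height AspectRatio reroute, exceeds 2, otherwise zero, limiting that result to ultrawide shapes.
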Material /MatchmakingSettingsFN1Shooters/MatchmakingSettings/Materials/M_UI_FN1_Vignettes
graph {
 properties {bAutomaticallySetUsageInEditor=false, blendMode=BLEND_Translucent, materialDomain=MD_UI, usedShadingModels=Default Lit}
 nodes [
  n1 = OneMinus(None=MaterialFunctionCall(UVs=MaterialFunctionCall() {materialFunction=&MF_UI_UVScaling}.ScaleX) {materialFunction=&RadialGradientExponential}.RadialGradientExponential)
  n2 = Constant()
  If(A=NamedRerouteUsage(), "A > B"=@n1, "A == B"=@n2, "A < B"=@n2) {constB=2}
  Multiply(A=@n1, B=Constant4Vector() {constant=<rgba 0.0012 0 0.0065 1>}.RGBA)
  n3 = MaterialFunctionCall() {materialFunction=&GetUserInterfaceUV}
  NamedRerouteDeclaration(None=Divide(A=ComponentMask(None=@n3.Pixel Size) {r=true}, B=ComponentMask(None=@n3.Pixel Size) {g=true})) {name=AspectRatio, nodeColor=<rgba 0 0.0859 1 1>}
 ]
}
