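# Error-state widget for FN1 matchmaking settings that shows a Ready Up error as icon and text in a width-clamped UIKit info box.
# Set Error collapses the widget for missing or empty errors, swaps in a restricted icon for content-gated errors, and starts a countdown view model when an expiration time exists.
# Set Error Text With Timer formats remaining time into the message.
# PreConstruct applies rich-text style, decorators, padding and icon tint.
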WidgetBlueprint /MatchmakingSettingsFN1Shooters/MatchmakingSettings/WBP_FN1_MatchmakingSettings_ErrorState
graphs [
 EventGraph {nodes [
   n1 = Event PreConstruct()
   > Parent: PreConstruct(IsDesignTime=@n1.IsDesignTime)
   > Set Error Widget Visual() {is=title K2Node_CallFunction}
  ]}
 Set Error {nodes [
   n1 = Set Error!() -> (then, ErrorUI:Ready Up Error UI Object Reference) {is=title K2Node_FunctionEntry}
   > n2 = Is Valid(InputObject=@n1)
   > n3 = Set ErrorText(ErrorText:Text=Error Text!(self:Ready Up Error UI Object Reference=Get ErrorUI() -> (ErrorUI:Ready Up Error UI Object Reference)) -> (ReturnValue:Text)) -> (Output_Get:Text)
   > n4 = Branch(Condition=NOT Boolean(A=TextIsEmpty(InText=@n3)))
   > SetVisibility(InVisibility=SelfHitTestInvisible)
   n5 = Expiration Time!(self:Ready Up Error UI Object Reference=Get ErrorUI()) -> (ReturnValue:"Date Time Structure (by ref)")
   > n6 = Branch(Condition="Not Equal (DateTime)"!(A:Date Time Structure=@n5, B:Date Time Structure) -> (ReturnValue:bool))
   > n7 = Setup Countdown Timer VM!(execute, self, DataTime:Date Time Structure=@n5) -> (then)
   n8 = Sequence(execute=[@n7.then, @n14.then])
   n9 = Get ErrorUI()
   > n10 = Cast To ContentGatedUI!(execute, Object=@n9) -> (then, CastFailed, AsContent Gated UI:Content Gated UI Object Reference) {is=title K2Node_DynamicCast}
   > Branch(Condition=Is Restricted!(self:Content Gated UI Object Reference=@n10) -> (ReturnValue:bool))
   > Set ErrorUI_Icon(ErrorUI_Icon:Texture Object Reference=Get ContentGated_Restricted_Icon() -> (ContentGated_Restricted_Icon:Texture Object Reference)) -> (Output_Get:Texture Object Reference)
   Set ErrorUI_Icon(execute=@n10.CastFailed, ErrorUI_Icon=Icon!(self:Ready Up Error UI Object Reference=@n9) -> (ReturnValue:Texture Object Reference))
   n11 = Get ErrorUI_Icon() -> (ErrorUI_Icon:Texture Object Reference)
   n12 = Is Valid(exec=@n8.then_1, InputObject=@n11)
   SetVisibility(
    execute=@n12.Is Not Valid
    self=Get Icon!(self:WBP UIKit Info Box Object Reference=Get WBP_UIKit_InfoBox() -> (WBP_UIKit_InfoBox:WBP UIKit Info Box Object Reference)) -> (Icon:Image Object Reference)
    InVisibility=Collapsed
   )
   n13 = CreateBrushFromAsset!(execute=@n12.Is Valid, LoadedAsset:Object Reference=@n11, BrushSize_X:double=40, BrushSize_Y:double=40) -> (then, ReturnValue:Slate Brush Structure)
   > SetBrushFromTexture!(
    execute
    self:Image Object Reference=Get Icon(self=Get WBP_UIKit_InfoBox())
    Texture:Texture 2D Object Reference=GetBrushResourceAsTexture2D!(Brush:"Slate Brush Structure (by ref)"=@n13) -> (ReturnValue:Texture 2D Object Reference)
    bMatchSize:bool=false
   ) -> (then)
   > SetVisibility(self=Get Icon(self=Get WBP_UIKit_InfoBox()), InVisibility=SelfHitTestInvisible)
   n14 = SetText(
    execute=@n6.else
    self=Get Text(self:WBP UIKit Info Box Object Reference=Get WBP_UIKit_InfoBox()) -> (Text:Common Rich Text Block Object Reference)
    InText=Get ErrorText() -> (ErrorText:Text)
   )
   SetVisibility(execute=[@n4.else, @n2.Is Not Valid], InVisibility=Collapsed)
   > Reroute Node()
  ]}
 Set Error Text With Timer {nodes [
   n1 = Set Error Text With Timer!() -> (then, TimeLeft:Text) {is=title K2Node_FunctionEntry}
   > SetText(self=Get Text(self=Get WBP_UIKit_InfoBox()), InText=Format Text!(Format=Get ErrorText(), TimeRemaining:Text=@n1) -> (Result))
  ]}
 Set Error Widget Visual {nodes [
   Set Error Widget Visual() {is=title K2Node_FunctionEntry}
   n1 = SlotAsOverlaySlot(Widget=Get HorizontalBox!(self:WBP UIKit Info Box Object Reference=Get WBP_UIKit_InfoBox()) -> (HorizontalBox:Horizontal Box Object Reference))
   > n2 = Is Valid(InputObject=@n1)
   > n3 = SetPadding!(
    execute
    self:Overlay Slot Object Reference=@n1
    InPadding_Left:float=0
    InPadding_Top:float=0
    InPadding_Right:float=0
    InPadding_Bottom:float=0
   ) -> (then)
   n4 = Get WBP_UIKit_InfoBox()
   n5 = Get Text(self=@n4)
   Set DefaultTextStyleOverride!(
    execute=[@n3.then, @n2.Is Not Valid]
    DefaultTextStyleOverride:Text Block Style Structure=Get RichTextStyleOverride() -> (RichTextStyleOverride:Text Block Style Structure)
    self:Rich Text Block Object Reference=@n5
   ) -> (then, Output_Get:Text Block Style Structure)
   > SetDecorators!(
    execute
    self:Rich Text Block Object Reference=@n5
    InDecoratorClasses:Array of Rich Text Block Decorator Class References=Get RichTextDecorator() -> (RichTextDecorator:Array of Rich Text Block Decorator Class References)
   ) -> (then)
   n6 = Get RichTextStyleOverride()
   > SetColorAndOpacity(
    self:Image Object Reference=Get Icon(self=@n4)
    InColorAndOpacity:Linear Color Structure=Break Slate Color!(SlateColor:"Slate Color Structure (by ref)"=Break Text Block Style!(TextBlockStyle:"Text Block Style Structure (by ref)"=@n6) -> (ColorAndOpacity:Slate Color Structure) {is=title K2Node_BreakStruct}) -> (SpecifiedColor:Linear Color Structure) {is=title K2Node_BreakStruct}
   )
  ]}
 SetMobileStyle {nodes [
   SetMobileStyle()
   > Branch(Condition=IsMobileGame())
   n1 = Get SB_MaxWidth() -> (SB_MaxWidth:Size Box Object Reference)
   > SetMinDesiredWidth!(execute, self:Size Box Object Reference=@n1, InMinDesiredWidth:float=464) -> (then)
   > SetMaxDesiredWidth!(execute, self:Size Box Object Reference=@n1, InMaxDesiredWidth:float=650) -> (then)
  ]}
 Setup Countdown Timer VM {nodes [
   n1 = Setup Countdown Timer VM!() -> (then, DataTime:Date Time Structure) {is=title K2Node_FunctionEntry}
   n2 = Get FortCountdownTimerVM() -> (FortCountdownTimerVM:Countdown Timer View Model Object Reference)
   > SetupViewModel!(
    execute
    self:Countdown Timer View Model Object Reference=@n2
    InFutureTime:"Date Time Structure (by ref)"=@n1
    InLiveText:"Text (by ref)"=Get In Live Text() -> (In Live Text:Text)
   ) -> (then)
   > SetTimerUnitDisplayType!(execute, self:Countdown Timer View Model Object Reference=@n2, InTimerUnitDisplayType:ETimespanUnitDisplayFormat Enum=Abbreviated) -> (then)
   > SetApproximateLongTimers!(execute, self:Countdown Timer View Model Object Reference=@n2, bInApproximateLongTimers:bool=false) -> (then)
   > StartTimer(self:Countdown Timer View Model Object Reference=@n2)
  ]}
]
variables [
 ErrorText {type=string, default=""}
 In Live Text {type=string, default=0}
]
widgets {
 info {inheritedWidgetCount=0, parentClass=&UserWidget, rootWidgetClass=&SizeBox, widgetCount=2}
 widgets [
  SB_MaxWidth {
   widgetClassPath=&SizeBox
   bIsVariable=true
   widget=SB_MaxWidth
   properties {bOverride_MaxDesiredWidth=true, bOverride_MinDesiredWidth=true, clipping=ClipToBounds, maxDesiredWidth=650, minDesiredWidth=464}
  }
  WBP_UIKit_InfoBox {
   widgetClassPath=/FortUILibrary/UIKit/InfoBox/WBP_UIKit_InfoBox
   bIsVariable=true
   parent=SB_MaxWidth
   widget=WBP_UIKit_InfoBox
   properties {
    iconMaterial=/FortUILibrary/Textures/MaterialInstances/MI_UIKit_Icon_Locked
    previewText="[TEST] Error text Error text Error text Error text"
    show Background=false
    use Preview Text=true
    iconSize=<40 40>
   }
   slot {class=SizeBoxSlot}
  }
 ]
}
