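# Polymorphic list view widget for FN1 Shooters matchmaking settings, showing an entry widget per setting view model.
# It picks each entry class from a discrete or scalar VM's variant via MatchmakingUISubsystem, loading it with a blocking load.
# Generated entries append their analytics attributes to an initial array, which has a reset function.
# GetFinalAnalyticsAttributes re-queries each item's widget through BPI_MatchmakingSettingsEntry to return current attributes.
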WidgetBlueprint /MatchmakingSettingsFN1Shooters/MatchmakingSettings/WBP_FN1_MatchmakingSettings_PolyList
graphs [
 EventGraph {nodes [
   n1 = "On Entry Generated (ListView)"!() -> (OutputDelegate, then, Widget:User Widget Object Reference) {is=title K2Node_ComponentBoundEvent}
   > n2 = Cast To BPI_MatchmakingSettingsEntry!(execute, Object=@n1.Widget) -> (then, CastFailed, AsBPI Matchmaking Settings Entry:BPI Matchmaking Settings Entry Interface) {is=title K2Node_DynamicCast}
   > n3 = GetAnalyticsAttribute!(execute, self:BPI Matchmaking Settings Entry Interface=@n2) -> (then, Attribute:Analytics Event Attr Structure)
   > Add(
    TargetArray:Array of Analytics Event Attr Structures=Get InitialAnalyticsAttributes() -> (InitialAnalyticsAttributes:Array of Analytics Event Attr Structures)
    NewItem:"Analytics Event Attr Structure (by ref)"=@n3
   ) {is=title K2Node_CallArrayFunction}
   Event PreConstruct()
   Event Construct()
   Event Tick()
  ]}
 GetDesiredEntryClassForItem {nodes [
   n1 = GetDesiredEntryClassForItem!() -> (then, Item:Object Reference) {is=title K2Node_FunctionEntry}
   > n2 = Cast To FortMatchmakingSettingDiscreteVM!(execute, Object=@n1) -> (then, CastFailed, AsFort Matchmaking Setting Discrete VM:Fort Matchmaking Setting Discrete VM Object Reference) {is=title K2Node_DynamicCast}
   > n3 = Set VariantKey(VariantKey:Name=GetVariant!(self:Fort Matchmaking Setting Discrete VM Object Reference=@n2) -> (ReturnValue:Name=None)) -> (Output_Get:Name=None)
   n4 = GetSettingEntryWidgetByForVMType!(
    self:Matchmaking UISubsystem Object Reference=Get MatchmakingUISubsystem() -> (ReturnValue:Matchmaking UISubsystem Object Reference) {is=title K2Node_GetSubsystem}
    VMType:Name=Get VariantKey() -> (VariantKey:Name=None)
   ) -> (OutWidgetClass:User Widget Soft Class Reference, ReturnValue:bool)
   n5 = Branch(execute=[@n8.then, @n3.then], Condition=@n4.ReturnValue)
   > n6 = LoadClassAsset_Blocking(AssetClass=@n4)
   > Return Node!(
    execute
    ReturnValue:User Widget Class Reference=Cast To UserWidget Class!(Class:Object Class Reference=@n6) -> (AsUser Widget:User Widget Class Reference, bSuccess:bool=~) {is=title K2Node_ClassDynamicCast}
   )
   Return Node(execute=[@n5.else, @n7.CastFailed])
   n7 = Cast To FortMatchmakingSettingScalarVM!(execute=@n2.CastFailed, Object=@n1) -> (then, CastFailed, AsFort Matchmaking Setting Scalar VM:Fort Matchmaking Setting Scalar VM Object Reference) {is=title K2Node_DynamicCast}
   > n8 = Set VariantKey(VariantKey=GetVariant(self:Fort Matchmaking Setting Scalar VM Object Reference=@n7) -> (ReturnValue=None)) -> (Output_Get=None)
  ]}
 GetFinalAnalyticsAttributes {nodes [
   GetFinalAnalyticsAttributes() {is=title K2Node_FunctionEntry}
   > n1 = For Each Loop(Array:Array of Object References=GetListItems!(self:List View Object Reference=Get ListView() -> (ListView:Common List View Object Reference)) -> (ReturnValue:Array of Object References)) -> (Array Element:Object Reference)
   > n2 = Cast To BPI_MatchmakingSettingsEntry(Object=GetWidgetFromItem!(self:Fort Polymorphic View Object Reference, Item:Object Reference=@n1) -> (ReturnValue:User Widget Object Reference))
   > n3 = GetAnalyticsAttribute(self=@n2)
   > Add(TargetArray=Get FinalAnalyticsAttributes() -> (FinalAnalyticsAttributes:Array of Analytics Event Attr Structures), NewItem=@n3)
   Return Node!(execute=@n1.Completed, Attributes:Array of Analytics Event Attr Structures=Get FinalAnalyticsAttributes())
  ]}
 ResetInitialAnalyticsAttributes {nodes [
   ResetInitialAnalyticsAttributes() {is=title K2Node_FunctionEntry}
   > Clear(TargetArray:Array of Analytics Event Attr Structures=Get InitialAnalyticsAttributes())
  ]}
]
variables [InitialAnalyticsAttributes {type="array<AnalyticsEventAttr>", default []}]
widgets {
 info {
  inheritedWidgetCount=1
  parentClass=/Script/FortniteUI.FortPolymorphicListView
  rootWidgetClass=&CommonListView
  widgetCount=1
 }
 widgets [ListView {
   widgetClassPath=&CommonListView
   bInherited=true
   bIsVariable=true
   widget=ListView
   properties {
    bReturnFocusToSelection=true
    clipping=Inherit
    entryWidgetClass=/MatchmakingSettingsFN1Shooters/MatchmakingSettings/Entries/Buttons/WBP_FN1_MatchmakingSettings_EntryBase.WBP_FN1_MatchmakingSettings_EntryBase_C
    numDesignerPreviewEntries=4
    verticalEntrySpacing=15
    scrollBarPadding=<ltrb 30 0 0 0>
    scrollBarStyle {
     thickness=12
     draggedThumbImage {
      bIsDynamicallyLoaded=false
      drawAs=Box
      imageType=FullColor
      mirroring=NoMirror
      resourceName=../../../Engine/Content/Slate/Common/Scrollbar_Thumb.png
      resourceObject=null
      tiling=NoTile
      imageSize=<0 0>
      margin=<ltrb 0.25 0.25 0.25 0.25>
      outlineSettings {
       bUseBrushTransparency=false
       roundingType=HalfHeightRadius
       width=0
       color {colorUseRule=UseColor_Specified, specifiedColor=<rgba 0 0 0 0>}
       cornerRadii=<0 0 0 0>
      }
      tintColor {colorUseRule=UseColor_Specified, specifiedColor=<rgba 1 1 1 1>}
      uVRegion {bIsValid=false, max=<0 0>, min=<0 0>}
     }
     horizontalBackgroundImage {
      bIsDynamicallyLoaded=false
      drawAs=NoDrawType
      imageType=NoImage
      mirroring=NoMirror
      resourceName=None
      resourceObject=null
      tiling=NoTile
      imageSize=<0 0>
      margin=<ltrb 0 0 0 0>
      outlineSettings {
       bUseBrushTransparency=false
       roundingType=HalfHeightRadius
       width=0
       color {colorUseRule=UseColor_Specified, specifiedColor=<rgba 0 0 0 0>}
       cornerRadii=<0 0 0 0>
      }
      tintColor {colorUseRule=UseColor_Specified, specifiedColor=<rgba 1 1 1 1>}
      uVRegion {bIsValid=false, max=<0 0>, min=<0 0>}
     }
     horizontalBottomSlotImage {
      bIsDynamicallyLoaded=false
      drawAs=Image
      imageType=FullColor
      mirroring=NoMirror
      resourceName=../../../Engine/Content/Slate/Common/Scrollbar_Background_Horizontal.png
      resourceObject=null
      tiling=NoTile
      imageSize=<8 8>
      margin=<ltrb 0 0 0 0>
      outlineSettings {
       bUseBrushTransparency=false
       roundingType=HalfHeightRadius
       width=0
       color {colorUseRule=UseColor_Specified, specifiedColor=<rgba 0 0 0 0>}
       cornerRadii=<0 0 0 0>
      }
      tintColor {colorUseRule=UseColor_Specified, specifiedColor=<rgba 1 1 1 1>}
      uVRegion {bIsValid=false, max=<0 0>, min=<0 0>}
     }
     horizontalTopSlotImage {
      bIsDynamicallyLoaded=false
      drawAs=Image
      imageType=FullColor
      mirroring=NoMirror
      resourceName=../../../Engine/Content/Slate/Common/Scrollbar_Background_Horizontal.png
      resourceObject=null
      tiling=NoTile
      imageSize=<8 8>
      margin=<ltrb 0 0 0 0>
      outlineSettings {
       bUseBrushTransparency=false
       roundingType=HalfHeightRadius
       width=0
       color {colorUseRule=UseColor_Specified, specifiedColor=<rgba 0 0 0 0>}
       cornerRadii=<0 0 0 0>
      }
      tintColor {colorUseRule=UseColor_Specified, specifiedColor=<rgba 1 1 1 1>}
      uVRegion {bIsValid=false, max=<0 0>, min=<0 0>}
     }
     hoveredThumbImage {
      bIsDynamicallyLoaded=false
      drawAs=Box
      imageType=FullColor
      mirroring=NoMirror
      resourceName=../../../Engine/Content/Slate/Common/Scrollbar_Thumb.png
      resourceObject=null
      tiling=NoTile
      imageSize=<0 0>
      margin=<ltrb 0.25 0.25 0.25 0.25>
      outlineSettings {
       bUseBrushTransparency=false
       roundingType=HalfHeightRadius
       width=0
       color {colorUseRule=UseColor_Specified, specifiedColor=<rgba 0 0 0 0>}
       cornerRadii=<0 0 0 0>
      }
      tintColor {colorUseRule=UseColor_Specified, specifiedColor=<rgba 1 1 1 1>}
      uVRegion {bIsValid=false, max=<0 0>, min=<0 0>}
     }
     normalThumbImage {
      bIsDynamicallyLoaded=false
      drawAs=Box
      imageType=FullColor
      mirroring=NoMirror
      resourceName=../../../Engine/Content/Slate/Common/Scrollbar_Thumb.png
      resourceObject=null
      tiling=NoTile
      imageSize=<0 0>
      margin=<ltrb 0.25 0.25 0.25 0.25>
      outlineSettings {
       bUseBrushTransparency=false
       roundingType=HalfHeightRadius
       width=0
       color {colorUseRule=UseColor_Specified, specifiedColor=<rgba 0 0 0 0>}
       cornerRadii=<0 0 0 0>
      }
      tintColor {colorUseRule=UseColor_Specified, specifiedColor=<rgba 1 1 1 1>}
      uVRegion {bIsValid=false, max=<0 0>, min=<0 0>}
     }
     verticalBackgroundImage {
      bIsDynamicallyLoaded=false
      drawAs=NoDrawType
      imageType=NoImage
      mirroring=NoMirror
      resourceName=None
      resourceObject=null
      tiling=NoTile
      imageSize=<0 0>
      margin=<ltrb 0 0 0 0>
      outlineSettings {
       bUseBrushTransparency=false
       roundingType=HalfHeightRadius
       width=0
       color {colorUseRule=UseColor_Specified, specifiedColor=<rgba 0 0 0 0>}
       cornerRadii=<0 0 0 0>
      }
      tintColor {colorUseRule=UseColor_Specified, specifiedColor=<rgba 1 1 1 1>}
      uVRegion {bIsValid=false, max=<0 0>, min=<0 0>}
     }
     verticalBottomSlotImage {
      bIsDynamicallyLoaded=false
      drawAs=Image
      imageType=FullColor
      mirroring=NoMirror
      resourceName=../../../Engine/Content/Slate/Common/Scrollbar_Background_Vertical.png
      resourceObject=null
      tiling=NoTile
      imageSize=<8 8>
      margin=<ltrb 0 0 0 0>
      outlineSettings {
       bUseBrushTransparency=false
       roundingType=HalfHeightRadius
       width=0
       color {colorUseRule=UseColor_Specified, specifiedColor=<rgba 0 0 0 0>}
       cornerRadii=<0 0 0 0>
      }
      tintColor {colorUseRule=UseColor_Specified, specifiedColor=<rgba 1 1 1 1>}
      uVRegion {bIsValid=false, max=<0 0>, min=<0 0>}
     }
     verticalTopSlotImage {
      bIsDynamicallyLoaded=false
      drawAs=Image
      imageType=FullColor
      mirroring=NoMirror
      resourceName=../../../Engine/Content/Slate/Common/Scrollbar_Background_Vertical.png
      resourceObject=null
      tiling=NoTile
      imageSize=<8 8>
      margin=<ltrb 0 0 0 0>
      outlineSettings {
       bUseBrushTransparency=false
       roundingType=HalfHeightRadius
       width=0
       color {colorUseRule=UseColor_Specified, specifiedColor=<rgba 0 0 0 0>}
       cornerRadii=<0 0 0 0>
      }
      tintColor {colorUseRule=UseColor_Specified, specifiedColor=<rgba 1 1 1 1>}
      uVRegion {bIsValid=false, max=<0 0>, min=<0 0>}
     }
    }
   }
  }]
}
