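# Translucent UI material for the FN1 Shooters matchmaking settings that draws a glowing button outline.
# A skewed-corner square SDF is anti-aliased, has a rerouted mask subtracted, and is modulated by a reflected gradient.
# The result is color-dodged over the Slate post buffer and added back.
# A padded vertex transform also rescales the UVs for a rounded inner-shape mask, and aspect-ratio reroutes derive from widget pixel size.
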Material /MatchmakingSettingsFN1Shooters/MatchmakingSettings/Materials/M_UI_FN1_ButtonOutline
graph {
 properties {bAutomaticallySetUsageInEditor=false, blendMode=BLEND_Translucent, materialDomain=MD_UI, usedShadingModels=Default Lit}
 nodes [
  n1 = Divide(A=Subtract(A=Constant() {r=1}, B=Constant() {r=0.9136}), B=NamedRerouteUsage())
  n2 = Constant() {r=1}
  n3 = Constant() {r=1}
  n4 = OneMinus(None=Divide(A=Subtract(A=@n2, B=@n3), B=NamedRerouteUsage()))
  n5 = Constant() {r=1}
  n6 = Constant() {r=0.912}
  n7 = OneMinus(None=Divide(A=Subtract(A=@n5, B=@n6), B=NamedRerouteUsage()))
  n8 = Divide(A=Subtract(A=Constant() {r=1}, B=Constant() {r=1}), B=NamedRerouteUsage())
  n9 = MaterialFunctionCall!(
   UVs
   Corner 01=AppendVector(A=OneMinus(None=@n1), B=Constant() {r=1})
   Corner 02=AppendVector(A=OneMinus(None=@n8), B=Constant() {r=-1})
   Corner 03=AppendVector(A=Multiply(A=@n4) {constB=-1}, B=Constant() {r=-0.9392})
   Corner 04=AppendVector(A=Multiply(A=@n7) {constB=-1}, B=Constant() {r=0.9376})
   Glow Min
   Glow Max
   Outline Thickness
  ) {materialFunction=/FortUILibrary/MaterialFunctions/Shapes/MF_UI_WIP_Square}
  n10 = MaterialFunctionCall!(UVs, Offset, RotationCenter, RotationAngle) {materialFunction=/FortUILibrary/MaterialFunctions/Gradients/MF_UI_ReflectedGradient}
  n11 = Multiply(
   A=Subtract(
    A=MaterialFunctionCall(DistanceField=@n9.SDF) {materialFunction=&MF_UI_AntiAliasing}.Inner
    B=NamedRerouteUsage()
   )
   B=Multiply(A=LinearInterpolate(Alpha=@n10.Result) {constA=0.7}) {constB=0.6}
  )
  Add(
   A=MaterialFunctionCall!(Blend=@n11, Base=MaterialFunctionCall!("UVs (Optional)") {materialFunction=/Engine/Functions/UserInterface/GetSlatePost0}.RGB) {materialFunction=/Engine/Functions/Engine_MaterialFunctions03/Blends/Blend_ColorDodge}.Result
   B=@n11
  )
  n12 = MaterialFunctionCall() {materialFunction=&GetUserInterfaceUV}
  NamedRerouteDeclaration(None=Divide(A=ComponentMask(None=@n12.Pixel Size) {g=true}, B=ComponentMask(None=@n12.Pixel Size) {r=true})) {name=AspectY, nodeColor=<rgba 0 0.7891 1 1>}
  n13 = TextureCoordinate()
  n14 = MaterialFunctionCall!(Anchor, Translation, Padding=AppendVector(A=Constant() {r=20}, B=Constant() {r=18}), Scale) {materialFunction=/FortUILibrary/MaterialFunctions/Basics/MF_UI_VertexTransform}
  n15 = Divide(B=@n14.Total Scale Factor)
  NamedRerouteDeclaration(None=MaterialFunctionCall(
    UV=MaterialFunctionCall(UVs=@n13, Texture Scale=@n15) {materialFunction=&ScaleUVsByCenter}.UVs
    Size=Constant2Vector() {g=1, r=1}
    Corner Radius=Constant()
   ) {materialFunction=&MF_UI_CoreUI_RoundedMask}.Shape Mask) {name=InnerShape, nodeColor=<rgba 0 0.4141 1 1>}
  NamedRerouteDeclaration(None=@n14.Transformed Position) {name=TransformPosition, nodeColor=<rgba 1 0.1172 0 1>}
  n16 = MaterialFunctionCall() {materialFunction=&GetUserInterfaceUV}
  NamedRerouteDeclaration(None=Divide(A=ComponentMask(None=@n16.Pixel Size) {r=true}, B=ComponentMask(None=@n16.Pixel Size) {g=true})) {name=AspectX, nodeColor=<rgba 0 0.7891 1 1>}
  NamedRerouteUsage()
 ]
}
