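# Button background block for the matchmaking settings UI, built on the UIKit modular button background base with a material-driven Fill and an Outline image.
# State events play animations forward or reversed, branching on selection, to drive Trans_* material scalars and Outline opacity.
# Hover also animates AnimPadding, whose setter scales Fill's horizontal padding by its cached width.
# Focus transitions are skipped on touch input, and instant select, enable and disable events stop running transitions.
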WidgetBlueprint /MatchmakingSettingsFN1Shooters/MatchmakingSettings/WBP_MatchmakingSettingsBackground
animations [
 Anim_Hovered_Transition {
  binding_count=2
  length_seconds=1
  object=&self:Anim_Hovered_Transition
  sequence {bindings [
    Fill {
     id=#1
     possessed_class=Image
     tracks [{
       type=MovieSceneWidgetMaterialTrack
       sections [{
         type=MovieSceneParameterSection
         end_frame=20
         start_frame=0
         channels [
          Trans_Hovered {type=MovieSceneScriptingFloatChannel, keys [[10,0],[19.9997,1]]}
          IdleTransition {type=MovieSceneScriptingFloatChannel, keys [[0,0],[10,1]]}
         ]
        }]
      }]
    }
    WBP_MatchmakingSettingsBackground_C_117 {
     id=#2
     possessed_class=WBP_MatchmakingSettingsBackground_C
     tracks [{
       type=MovieSceneDoubleTrack
       display_name=Anim Padding
       sections [{type=MovieSceneDoubleSection, end_frame=20, start_frame=0, channels [None {type=MovieSceneScriptingDoubleChannel, keys [[0,0],[20,1]]}]}]
      }]
    }
   ]}
 }
 Anim_Focused_Transition {
  binding_count=3
  length_seconds=1
  object=&self:Anim_Focused_Transition
  sequence {bindings [
    Fill {
     id=#3
     possessed_class=Image
     tracks [{
       type=MovieSceneWidgetMaterialTrack
       sections [{
         type=MovieSceneParameterSection
         end_frame=20
         start_frame=0
         channels [Trans_Focused {type=MovieSceneScriptingFloatChannel, keys [[0,0],[20,1]]}]
        }]
      }]
    }
    Outline {
     id=#4
     possessed_class=Image
     tracks [{
       type=MovieSceneFloatTrack
       display_name=Render Opacity
       sections [{type=MovieSceneFloatSection, end_frame=20, start_frame=0, channels [None {type=MovieSceneScriptingFloatChannel, keys [[0,0],[20,1]]}]}]
      }]
    }
   ]}
 }
 Anim_Pressed_Transition {
  binding_count=1
  length_seconds=1
  object=&self:Anim_Pressed_Transition
  sequence {bindings [Fill {
     id=#5
     possessed_class=Image
     tracks [{
       type=MovieSceneWidgetMaterialTrack
       sections [{
         type=MovieSceneParameterSection
         end_frame=20
         start_frame=0
         channels [Trans_Pressed {type=MovieSceneScriptingFloatChannel, keys [[0,0],[20,1]]}]
        }]
      }]
    }]}
 }
 Anim_Disabled_Transition {
  binding_count=1
  length_seconds=1
  object=&self:Anim_Disabled_Transition
  sequence {bindings [Fill {
     id=#6
     possessed_class=Image
     tracks [{
       type=MovieSceneWidgetMaterialTrack
       sections [{
         type=MovieSceneParameterSection
         end_frame=20
         start_frame=0
         channels [Trans_Disabled {type=MovieSceneScriptingFloatChannel, keys [[0,0],[20,1]]}]
        }]
      }]
    }]}
 }
 Anim_Selected_Transition {
  binding_count=1
  length_seconds=1
  object=&self:Anim_Selected_Transition
  sequence {bindings [Fill {
     id=#7
     possessed_class=FortMobileImage
     tracks [{
       type=MovieSceneWidgetMaterialTrack
       sections [{
         type=MovieSceneParameterSection
         end_frame=20
         start_frame=0
         channels [Trans_SelectedIdle {type=MovieSceneScriptingFloatChannel, default=0.6083, keys [[0,0],[20,1]]}]
        }]
      }]
    }]}
 }
 Anim_SelectedHovered_Transition {
  binding_count=2
  length_seconds=1
  object=&self:Anim_SelectedHovered_Transition
  sequence {bindings [
    Outline {
     id=#8
     possessed_class=Image
     tracks [{
       type=MovieSceneFloatTrack
       display_name=Render Opacity
       sections [{type=MovieSceneFloatSection, end_frame=20, start_frame=0, channels [None {type=MovieSceneScriptingFloatChannel, keys [[0,0],[20,1]]}]}]
      }]
    }
    Fill {
     id=#9
     possessed_class=FortMobileImage
     tracks [{
       type=MovieSceneWidgetMaterialTrack
       sections [{
         type=MovieSceneParameterSection
         end_frame=20
         start_frame=0
         channels [Trans_SelectedHovered {type=MovieSceneScriptingFloatChannel, default=0, keys [[0,0],[20,1]]}]
        }]
      }]
    }
   ]}
 }
 Anim_SelectedFocused_Transition {
  binding_count=2
  length_seconds=1
  object=&self:Anim_SelectedFocused_Transition
  sequence {bindings [
    Outline {
     id=#10
     possessed_class=Image
     tracks [{
       type=MovieSceneFloatTrack
       display_name=Render Opacity
       sections [{type=MovieSceneFloatSection, end_frame=20, start_frame=0, channels [None {type=MovieSceneScriptingFloatChannel, keys [[0,0],[20,1]]}]}]
      }]
    }
    Fill {
     id=#11
     possessed_class=FortMobileImage
     tracks [{
       type=MovieSceneWidgetMaterialTrack
       sections [{
         type=MovieSceneParameterSection
         end_frame=20
         start_frame=0
         channels [Trans_SelectedFocused {type=MovieSceneScriptingFloatChannel, default=0, keys [[0,0],[20,1]]}]
        }]
      }]
    }
   ]}
 }
 Anim_SelectedPressed_Transition {
  binding_count=1
  length_seconds=1
  object=&self:Anim_SelectedPressed_Transition
  sequence {bindings [Fill {
     id=#12
     possessed_class=FortMobileImage
     tracks [{
       type=MovieSceneWidgetMaterialTrack
       sections [{
         type=MovieSceneParameterSection
         end_frame=20
         start_frame=0
         channels [Trans_SelectedPressed {type=MovieSceneScriptingFloatChannel, default=0, keys [[0,0],[20,1]]}]
        }]
      }]
    }]}
 }
 Anim_SelectedDisabled_Transition {
  binding_count=1
  length_seconds=1
  object=&self:Anim_SelectedDisabled_Transition
  sequence {bindings [Fill {
     id=#13
     possessed_class=FortMobileImage
     tracks [{
       type=MovieSceneWidgetMaterialTrack
       sections [{
         type=MovieSceneParameterSection
         end_frame=20
         start_frame=0
         channels [Trans_SelectedDisabled {type=MovieSceneScriptingFloatChannel, default=0, keys [[0,0],[20,1]]}]
        }]
      }]
    }]}
 }
 Anim_SelectedFocusedWithoutTransform_Transition {
  binding_count=2
  length_seconds=1
  object=&self:Anim_SelectedFocusedWithoutTransform_Transition
  sequence {bindings [
    Fill {
     id=#11
     possessed_class=FortMobileImage
     tracks [{
       type=MovieSceneWidgetMaterialTrack
       sections [{
         type=MovieSceneParameterSection
         end_frame=20
         start_frame=0
         channels [Trans_SelectedFocused {type=MovieSceneScriptingFloatChannel, default=0, keys [[0,0],[20,1]]}]
        }]
      }]
    }
    Outline {
     id=#14
     possessed_class=Image
     tracks [{
       type=MovieSceneFloatTrack
       display_name=Render Opacity
       sections [{type=MovieSceneFloatSection, end_frame=20, start_frame=0, channels [None {type=MovieSceneScriptingFloatChannel, keys [[0,0],[20,1]]}]}]
      }]
    }
   ]}
 }
]
graphs [
 EventGraph {nodes [
   Event OnInstantTransitionToSelected() {is=title K2Node_Event}
   > Branch(Condition=GetIsSelectable!() -> (IsSelectable:bool))
   > Stop All Transitions!(In:Exec) -> (Out:Exec) {is=title K2Node_MacroInstance}
   > InstantTransitionToSelected!(
    execute
    MID:Material Instance Dynamic Object Reference=Get BackgroundDynamicMaterial() -> (BackgroundDynamicMaterial:Material Instance Dynamic Object Reference)
   ) -> (then)
   Event OnTransitionSelectedRequested() {is=title K2Node_Event}
   > Branch(Condition=GetIsSelectable())
   > n1 = Branch(Condition=GetIsFocused!() -> (IsFocused:bool))
   > PlayTransition!(
    execute
    self:UIKit Block Object Reference
    Animation:Widget Animation Object Reference=Get Anim_SelectedFocusedWithoutTransform_Transition() -> (Anim_SelectedFocusedWithoutTransform_Transition:Widget Animation Object Reference)
    bPlayForward:bool=true
   ) -> (then)
   PlayTransition(
    execute=@n1.else
    Animation=Get Anim_Selected_Transition() -> (Anim_Selected_Transition:Widget Animation Object Reference)
    bPlayForward=true
   )
   Event OnTransitionUnhoveredRequested() {is=title K2Node_Event}
   > SetIsFocused!(execute, self:WBP UIKit Block Base Object Reference, IsFocused:bool=false) -> (then)
   > n2 = Branch(Condition=GetIsSelected!() -> (IsSelected:bool))
   > SetScalarParameterValue(self=Get BackgroundDynamicMaterial(), ParameterName=Trans_SelectedIdle, Value=1)
   > PlayTransition(Animation=Get Anim_SelectedHovered_Transition() -> (Anim_SelectedHovered_Transition:Widget Animation Object Reference), bPlayForward=false)
   StopUnfinishedAnimation!(
    execute=@n2.else
    Target:User Widget Object Reference=Self-Reference()
    Animation:Widget Animation Object Reference=Get Anim_Selected_Transition()
   ) -> (then)
   > PlayTransition(Animation=Get Anim_Hovered_Transition() -> (Anim_Hovered_Transition:Widget Animation Object Reference), bPlayForward=false)
   Event OnTransitionFocusedRequested() {is=title K2Node_Event}
   > n3 = Branch(Condition=IsUsingInputType!(InputType:ECommonInputType Enum=Touch, PlayerController:Player Controller Object Reference=GetOwningPlayer()) -> (ReturnValue:bool))
   UpdateFillImageSize(execute=@n3.else)
   > SetIsFocused(IsFocused=true)
   > n4 = Branch(Condition=GetIsSelected())
   > PlayTransition(Animation=Get Anim_SelectedFocused_Transition() -> (Anim_SelectedFocused_Transition:Widget Animation Object Reference), bPlayForward=true)
   PlayTransition(execute=@n4.else, Animation=Get Anim_Focused_Transition() -> (Anim_Focused_Transition:Widget Animation Object Reference), bPlayForward=true)
   Event OnTransitionUnfocusedRequested() {is=title K2Node_Event}
   > n5 = Branch(Condition=IsUsingInputType(InputType=Touch, PlayerController=GetOwningPlayer()))
   SetIsFocused(execute=@n5.else, IsFocused=false)
   > n6 = Branch(Condition=GetIsSelected())
   > SetScalarParameterValue(self=Get BackgroundDynamicMaterial(), ParameterName=Trans_SelectedIdle, Value=1)
   > PlayTransition(Animation=Get Anim_SelectedFocused_Transition(), bPlayForward=false)
   StopUnfinishedAnimation(execute=@n6.else, Target=Self-Reference(), Animation=Get Anim_Selected_Transition())
   > PlayTransition(Animation=Get Anim_Focused_Transition(), bPlayForward=false)
   Event OnTransitionDisabledRequested() {is=title K2Node_Event}
   > n7 = Branch(Condition=GetIsSelected())
   > PlayTransition(Animation=Get Anim_SelectedDisabled_Transition() -> (Anim_SelectedDisabled_Transition:Widget Animation Object Reference), bPlayForward=true)
   PlayTransition(
    execute=@n7.else
    Animation=Get Anim_Disabled_Transition() -> (Anim_Disabled_Transition:Widget Animation Object Reference)
    bPlayForward=true
   )
   Event OnTransitionEnabledRequested() {is=title K2Node_Event}
   > n8 = Branch(Condition=GetIsSelected())
   > PlayTransition(Animation=Get Anim_SelectedDisabled_Transition(), bPlayForward=false)
   PlayTransition(execute=@n8.else, Animation=Get Anim_Disabled_Transition(), bPlayForward=false)
   Event OnTransitionHoveredRequested() {is=title K2Node_Event}
   > UpdateFillImageSize()
   > SetIsFocused(IsFocused=true)
   > n9 = Branch(Condition=GetIsSelected())
   > PlayTransition(Animation=Get Anim_SelectedHovered_Transition(), bPlayForward=true)
   PlayTransition(execute=@n9.else, Animation=Get Anim_Hovered_Transition(), bPlayForward=true)
   n10 = Event OnInitialized()
   SetBackgroundMaterial!(
    execute=[@n10.then, @n12.then]
    self:WBP UIKit Block Background Base Object Reference
    Material Instance:"Material Instance Object Reference (by ref)"=Get BackgroundMaterial() -> (BackgroundMaterial:Material Instance Object Reference)
   ) -> (then)
   Event OnTransitionDeselectedFocusedRequested() {is=title K2Node_Event}
   > Branch(Condition=GetIsSelectable())
   > PlayTransition(Animation=Get Anim_SelectedFocusedWithoutTransform_Transition(), bPlayForward=false)
   > StopUnfinishedAnimation(Target=Self-Reference(), Animation=Get Anim_Selected_Transition())
   > TransitionToDeselectedFocused!(execute, MID:Material Instance Dynamic Object Reference=Get BackgroundDynamicMaterial()) -> (then)
   n11 = Event PreConstruct()
   > n12 = Branch(Condition=@n11.IsDesignTime)
   Event OnTransitionDeselectedIdleRequested() {is=title K2Node_Event}
   > Branch(Condition=GetIsSelectable())
   > PlayTransition(Animation=Get Anim_Selected_Transition(), bPlayForward=false)
   > TransitionToDeselectedIdle!(execute, MID:Material Instance Dynamic Object Reference=Get BackgroundDynamicMaterial()) -> (then)
   Event OnInstantTransitionToEnabled() {is=title K2Node_Event}
   n13 = Get BackgroundDynamicMaterial()
   > n14 = Get Scalar Parameter Value!(execute, self:Material Instance Dynamic Object Reference=@n13, ParameterName:Name=Trans_SelectedDisabled) -> (then, ReturnValue:float) {is=title K2Node_CallFunction}
   > n15 = Get Scalar Parameter Value(self=@n13, ParameterName=Trans_Disabled)
   > Branch(Condition="float > float"(A=float + float(A=@n15, B=@n14), B=0.1))
   > Stop All Transitions()
   > InstantTransitionToEnabled!(execute, MID:Material Instance Dynamic Object Reference=@n13, IsSelected:bool=GetIsSelected()) -> (then)
   Event OnInstantTransitionToDisabled() {is=title K2Node_Event}
   > Stop All Transitions()
   > InstantTransitionToDisabled!(execute, MID:Material Instance Dynamic Object Reference=Get BackgroundDynamicMaterial(), IsSelected:bool=GetIsSelected()) -> (then)
  ]}
 ResetAllStates {nodes [
   ResetAllStates() {is=title K2Node_FunctionEntry}
   > Stop All Transitions()
   > SetRenderOpacity(self=Get Outline() -> (Outline:Image Object Reference), InOpacity=0)
   n1 = Get BackgroundDynamicMaterial()
   > SetScalarParameterValue(self=@n1, ParameterName=Trans_Hovered, Value=0)
   > SetScalarParameterValue(self=@n1, ParameterName=Trans_Focused, Value=0)
   > SetScalarParameterValue(self=@n1, ParameterName=Trans_Pressed, Value=0)
   > SetScalarParameterValue(self=@n1, ParameterName=Trans_Disabled, Value=0)
   > SetScalarParameterValue(self=@n1, ParameterName=Trans_SelectedIdle, Value=0)
   > SetScalarParameterValue(self=@n1, ParameterName=Trans_SelectedHovered, Value=0)
   > SetScalarParameterValue(self=@n1, ParameterName=Trans_SelectedFocused, Value=0)
   > SetScalarParameterValue(self=@n1, ParameterName=Trans_SelectedPressed, Value=0)
   > SetScalarParameterValue(self=@n1, ParameterName=Trans_SelectedFocused, Value=0)
   > SetScalarParameterValue(self=@n1, ParameterName=Trans_SelectedDisabled, Value=0)
  ]}
 SetAnimPadding {nodes [
   n1 = SetAnimPadding!() -> (then, AnimPadding:double=~) {is=title K2Node_FunctionEntry}
   n2 = Lerp(A=0, B=Get DefaultPaddings() -> (DefaultPaddings:double), Alpha=@n1)
   n3 = Get FillSize() -> (FillSize:Vector 2D Structure)
   n4 = Power!(Base:double=float / float(A=BreakVector2D(InVec=@n3), B=64), Exp:double=0.6) -> (ReturnValue:double) {is=title K2Node_CallFunction}
   n5 = Get DefaultPaddings()
   n6 = Lerp(
    A=0
    B="Min (Float)"(A:double=float * float(A=@n4, B=@n5) {is=title K2Node_PromotableOperator}, B:double=0) -> (ReturnValue:double) {is=title K2Node_CommutativeAssociativeBinaryOperator}
    Alpha=@n1
   )
   > SetPadding!(
    execute
    self:Overlay Slot Object Reference=SlotAsOverlaySlot(Widget=Get Fill() -> (Fill:Image Object Reference))
    InPadding_Left:float=@n6
    InPadding_Top:float=@n2
    InPadding_Right:float=@n6
    InPadding_Bottom:float=@n2
   ) -> (then)
  ]}
 SetBackgroundBrush {nodes [
   n1 = SetBackgroundBrush!() -> (then, InBrush:"Slate Brush Structure (by ref)") {is=title K2Node_FunctionEntry}
   n2 = Get Fill()
   > SetBrush(self=@n2, InBrush=@n1)
   > n3 = GetDynamicMaterial(self=@n2)
   > Set BackgroundDynamicMaterial(BackgroundDynamicMaterial:Material Instance Dynamic Object Reference=@n3) -> (Output_Get:Material Instance Dynamic Object Reference)
   > Return Node()
  ]}
 SetBackgroundMaterial {nodes [
   n1 = SetBackgroundMaterial!() -> (then, Material Instance:"Material Instance Object Reference (by ref)") {is=title K2Node_FunctionEntry}
   n2 = Get Fill()
   > SetBrushFromMaterial(self=@n2, Material=@n1)
   > n3 = GetDynamicMaterial(self=@n2)
   > n4 = Set BackgroundDynamicMaterial(BackgroundDynamicMaterial=@n3)
   > RefreshBackgroundMID!(execute, self:WBP UIKit Block Background Base Object Reference, Background MID:Material Instance Dynamic Object Reference=@n4) -> (then)
   > n5 = Branch(Condition=IsValid(Object=@n1))
   n6 = Get Fill()
   > n7 = SetVisibility(self=@n6, InVisibility=Visible)
   Return Node(execute=[@n8.then, @n7.then])
   n8 = SetVisibility(execute=@n5.else, self=@n6, InVisibility=Collapsed)
  ]}
 UpdateFillImageSize {nodes [
   UpdateFillImageSize() {is=title K2Node_FunctionEntry}
   n1 = Get Fill()
   > Set FillSize(FillSize:Vector 2D Structure=GetLocalSize!(Geometry:"Geometry Structure (by ref)"=GetPaintSpaceGeometry!(self:Widget Object Reference=@n1) -> (ReturnValue:"Geometry Structure (by ref)")) -> (ReturnValue:Vector 2D Structure)) -> (Output_Get:Vector 2D Structure)
  ]}
]
variables [
 AnimPadding {type=number}
 FillSize {type=Vector2D, default=<0 0>}
 PaddingIndex {type=number}
]
widgets {
 info {
  inheritedWidgetCount=0
  parentClass=/FortUILibrary/Systems/ModularButton/Blocks/WBP_UIKit_BlockBackground_Base.WBP_UIKit_BlockBackground_Base_C
  rootWidgetClass=&Overlay
  widgetCount=3
 }
 widgets [
  Overlay_0 {widgetClassPath=&Overlay, widget=Overlay_0}
  Fill {
   widgetClassPath=&Image
   bIsVariable=true
   parent=Overlay_0
   widget=Fill
   properties {brush {
     bIsDynamicallyLoaded=false
     drawAs=Image
     imageType=FullColor
     mirroring=NoMirror
     resourceName=None
     resourceObject=/MatchmakingSettingsFN1Shooters/MatchmakingSettings/Materials/MI_UI_MMSettingsComponentBackground
     tiling=NoTile
     imageSize=<32 32>
     margin=<ltrb 0.5 0.5 0.5 0.5>
     outlineSettings {
      bUseBrushTransparency=false
      roundingType=HalfHeightRadius
      width=0
      color {colorUseRule=UseColor_Specified, specifiedColor=<rgba 0 0 0 0>}
      cornerRadii=<0 0 0 0>
     }
     tintColor {colorUseRule=UseColor_Specified, specifiedColor=<rgba 1 1 1 1>}
     uVRegion {bIsValid=false, max=<0 0>, min=<0 0>}
    }}
   slot {class=OverlaySlot, properties {horizontalAlignment=HAlign_Fill, verticalAlignment=VAlign_Fill}}
  }
  Outline {
   widgetClassPath=&Image
   bIsVariable=true
   parent=Overlay_0
   widget=Outline
   properties {
    renderOpacity=0
    brush {
     bIsDynamicallyLoaded=false
     drawAs=Image
     imageType=FullColor
     mirroring=NoMirror
     resourceName=None
     resourceObject=/MatchmakingSettingsFN1Shooters/MatchmakingSettings/Materials/M_UI_FN1_ButtonOutline
     tiling=NoTile
     imageSize=<32 32>
     margin=<ltrb 0 0 0 0>
     outlineSettings {
      bUseBrushTransparency=false
      roundingType=HalfHeightRadius
      width=0
      color {colorUseRule=UseColor_Specified, specifiedColor=<rgba 0 0 0 0>}
      cornerRadii=<0 0 0 0>
     }
     tintColor {colorUseRule=UseColor_Specified, specifiedColor=<rgba 1 1 1 1>}
     uVRegion {bIsValid=false, max=<0 0>, min=<0 0>}
    }
   }
   slot {class=OverlaySlot, properties {horizontalAlignment=HAlign_Fill, verticalAlignment=VAlign_Fill, padding=<ltrb -4 -4 -4 -4>}}
  }
 ]
}
