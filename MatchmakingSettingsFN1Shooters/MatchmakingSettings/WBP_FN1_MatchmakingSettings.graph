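# Activatable FN1 shooters matchmaking settings screen, with a polymorphic settings list, Play and Done buttons, and game art, logo, map rotation and player count for the selected island.
# SetIsRanked drives a ranked look through a material parameter collection, the Play button and a pattern overlay, animating it on every switch after the first.
# Closing reports initial versus final option analytics and whether Play was clicked.
# Compact UI adjusts padding and alignment, and gamepad input refreshes focus.
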WidgetBlueprint /MatchmakingSettingsFN1Shooters/MatchmakingSettings/WBP_FN1_MatchmakingSettings
animations [
 Anim_Intro {
  binding_count=7
  length_seconds=1.0667
  object=&self:Anim_Intro
  sequence {
   bindings [
    MapRotation_Entry {
     id=#1
     possessed_class=WBP_FN1_MatchmakingSettings_MapRotation_Entry_C
     tracks [
      {
       type=MovieScene2DTransformTrack
       sections [{
         type=MovieScene2DTransformSection
         end_frame=60
         start_frame=1
         channels [Translation.Y {type=MovieSceneScriptingFloatChannel, keys [[1,100],[60,0]]}]
        }]
      }
      {
       type=MovieSceneFloatTrack
       display_name=Render Opacity
       sections [{type=MovieSceneFloatSection, end_frame=60, start_frame=1, channels [None {type=MovieSceneScriptingFloatChannel, keys [[1,0],[60,1]]}]}]
      }
     ]
    }
    BackgroundCMS {
     id=#2
     possessed_class=EpicCMSImage
     tracks [{
       type=MovieSceneWidgetMaterialTrack
       sections [{
         type=MovieSceneParameterSection
         end_frame=60
         start_frame=1
         channels [Transition {type=MovieSceneScriptingFloatChannel, keys [[1,0],[60,1]]}]
        }]
      }]
    }
    Button_Matchmaking {
     id=#3
     possessed_class=WBP_UI_Button_Matchmaking_C
     tracks [
      {
       type=MovieSceneFloatTrack
       display_name=Render Opacity
       sections [{type=MovieSceneFloatSection, end_frame=65, start_frame=1, channels [None {type=MovieSceneScriptingFloatChannel, keys [[1,0],[6,0],[65,1]]}]}]
      }
      {
       type=MovieScene2DTransformTrack
       sections [{
         type=MovieScene2DTransformSection
         end_frame=65
         start_frame=1
         channels [Translation.Y {type=MovieSceneScriptingFloatChannel, keys [[1,100],[6,100],[65,0]]}]
        }]
      }
     ]
    }
    O_ListClipping {
     id=#4
     possessed_class=Overlay
     tracks [
      {
       type=MovieSceneFloatTrack
       display_name=Render Opacity
       sections [{type=MovieSceneFloatSection, end_frame=60, start_frame=1, channels [None {type=MovieSceneScriptingFloatChannel, keys [[1,0],[60,1]]}]}]
      }
      {
       type=MovieScene2DTransformTrack
       sections [{
         type=MovieScene2DTransformSection
         end_frame=60
         start_frame=1
         channels [Translation.Y {type=MovieSceneScriptingFloatChannel, keys [[1,100],[60,0]]}]
        }]
      }
     ]
    }
    Button_Close {
     id=#5
     possessed_class=WBP_UIKit_Button_Regular_C
     tracks [
      {
       type=MovieSceneFloatTrack
       display_name=Render Opacity
       sections [{type=MovieSceneFloatSection, end_frame=65, start_frame=1, channels [None {type=MovieSceneScriptingFloatChannel, keys [[1,0],[6,0],[65,1]]}]}]
      }
      {
       type=MovieScene2DTransformTrack
       sections [{
         type=MovieScene2DTransformSection
         end_frame=65
         start_frame=1
         channels [Translation.Y {type=MovieSceneScriptingFloatChannel, keys [[1,99.6631],[6,100],[65,0]]}]
        }]
      }
     ]
    }
    Vignettes {
     id=#6
     possessed_class=Image
     tracks [{
       type=MovieSceneFloatTrack
       display_name=Render Opacity
       sections [{type=MovieSceneFloatSection, end_frame=60, start_frame=1, channels [None {type=MovieSceneScriptingFloatChannel, keys [[1,0],[60,1]]}]}]
      }]
    }
    WBP_FN1_MatchmakingSettings_ErrorState {
     id=#7
     possessed_class=WBP_FN1_MatchmakingSettings_ErrorState_C
     tracks [
      {
       type=MovieSceneFloatTrack
       display_name=Render Opacity
       sections [{type=MovieSceneFloatSection, end_frame=65, start_frame=1, channels [None {type=MovieSceneScriptingFloatChannel, keys [[1,0],[6,0],[65,1]]}]}]
      }
      {
       type=MovieScene2DTransformTrack
       sections [{
         type=MovieScene2DTransformSection
         end_frame=65
         start_frame=1
         channels [Translation.Y {type=MovieSceneScriptingFloatChannel, keys [[1,100],[6,100],[65,0]]}]
        }]
      }
     ]
    }
   ]
   display_rate {numerator=60}
  }
 }
 Anim_SwitchRank {
  binding_count=2
  length_seconds=0.35
  object=&self:Anim_SwitchRank
  sequence {bindings [
    WBP_FN1_MatchmakingSettings_C_54 {
     id=#8
     possessed_class=WBP_FN1_MatchmakingSettings_C
     tracks [{
       type=MovieSceneDoubleTrack
       display_name=Rank Transition
       sections [{
         type=MovieSceneDoubleSection
         end_frame=7
         start_frame=0
         channels [None {type=MovieSceneScriptingDoubleChannel, default=0, keys [[0,0],[7,1]]}]
        }]
      }]
    }
    RankedPattern {
     id=#9
     possessed_class=Image
     tracks [{
       type=MovieSceneFloatTrack
       display_name=Render Opacity
       sections [{
         type=MovieSceneFloatSection
         end_frame=7
         start_frame=0
         channels [None {type=MovieSceneScriptingFloatChannel, default=0, keys [[0,0],[7,1]]}]
        }]
      }]
    }
   ]}
 }
]
graphs [
 BP_GetDesiredFocusTarget {nodes [
   Get Desired Focus Target()
   n1 = Get ListView!(self:Fort Polymorphic List View Object Reference=Get PolyList() -> (PolyList:WBP FN1 Matchmaking Settings Poly List Object Reference)) -> (ListView:Common List View Object Reference)
   > SetSelectedIndex!(execute, self:List View Object Reference=@n1, Index:int=0) -> (then)
   > Return Node!(execute, ReturnValue:Widget Object Reference=@n1)
  ]}
 ButtonsUpNavigation {nodes [
   ButtonsUpNavigation!() -> (then, Navigation:EUINavigation Enum) {is=title K2Node_FunctionEntry}
   > Return Node(ReturnValue=Get ListView(self=Get PolyList()))
  ]}
 EventGraph {nodes [
   "On Clicked (Button_Matchmaking)"!() -> (OutputDelegate, then, Button:Common Button Base Object Reference) {is=title K2Node_ComponentBoundEvent}
   > n1 = Set IsPlayClicked(IsPlayClicked:bool=true) -> (Output_Get:bool)
   DeactivateWidget(execute=[@n5.then, @n1.then])
   > n2 = Cast To FortPlayerController(Object=GetPlayerController!(PlayerIndex:int=0) -> (ReturnValue:Player Controller Object Reference))
   n3 = Get PolyList()
   > n4 = GetFinalAnalyticsAttributes!(execute, self:WBP FN1 Matchmaking Settings Poly List Object Reference=@n3) -> (then, Attributes:Array of Analytics Event Attr Structures)
   > Fire Event Matchmaking Settings Update!(
    execute
    FortPC:Fort Player Controller Object Reference=@n2
    InitialOptions:Array of Analytics Event Attr Structures=Get InitialAnalyticsAttributes!(self:WBP FN1 Matchmaking Settings Poly List Object Reference=@n3) -> (InitialAnalyticsAttributes:Array of Analytics Event Attr Structures)
    FinalOptions:Array of Analytics Event Attr Structures=@n4
    bResult:bool=Get IsPlayClicked() -> (IsPlayClicked:bool)
   ) -> (then)
   > StopAllAnimations(self:User Widget Object Reference)
   "On Clicked (Button_Close)"!() -> (OutputDelegate, then, Button:Common Button Base Object Reference) {is=title K2Node_ComponentBoundEvent}
   > n5 = Set IsPlayClicked(IsPlayClicked=false)
   n6 = Event PreConstruct()
   > Parent: PreConstruct(IsDesignTime=@n6.IsDesignTime)
   > SetBlocksTiming!(
    execute
    self:UIKit Modular Button Object Reference=Get Button_Close() -> (Button_Close:WBP UIKit Button Generic Object Reference)
    Timing:"UIKit Block Timing Structure (by ref)"=Make UIKit Block Timing!(
     Hovering:float=0
     Unhovering:float=0
     Focusing:float=0
     Unfocusing:float=0
     Pressing:float=0
     Releasing:float=0
     Disabling:float=0
     Enabling:float=0
     Selecting:float=0
     DeselectingIdle:float=0
     DeselectingFocused:float=0
    ) -> (UIKitBlockTiming:UIKit Block Timing Structure) {is=title K2Node_MakeStruct}
   ) -> (then)
   > SetMobileStyle() {is=title K2Node_CallFunction}
   > Set DontPlayRankTransition(DontPlayRankTransition:bool=true) -> (Output_Get:bool)
   Event Construct()
   > PlayAnimation(InAnimation=Get Anim_Intro() -> (Anim_Intro:Widget Animation Object Reference))
   Event On Activated()
   > PlaySound2D(Sound=/Game/Sounds/Fort_UI_Sounds/FrontEnd/New_FrontEnd/Ranked/Fort_UI_Ranked_Whoosh_Match_Menu_Intro_Cue)
   > Set SelectedIslandViewModel(SelectedIslandViewModel:Selected Island View Model Object Reference=GetSelectedIslandViewModel!(self:Matchmaking UISubsystem Object Reference=Get MatchmakingUISubsystem() -> (ReturnValue:Matchmaking UISubsystem Object Reference) {is=title K2Node_GetSubsystem}) -> (ReturnValue:Selected Island View Model Object Reference)) -> (Output_Get:Selected Island View Model Object Reference)
   > Set SelectedIslandViewModel!(
    execute
    SelectedIslandViewModel=Get SelectedIslandViewModel() -> (SelectedIslandViewModel:Selected Island View Model Object Reference)
    self:WBP FN1 Matchmaking Settings Map Rotation Entry Object Reference=Get MapRotation_Entry() -> (MapRotation_Entry:WBP FN1 Matchmaking Settings Map Rotation Entry Object Reference)
   ) -> (then, Output_Get)
   n7 = Get CommonInputSubsystem()
   > Bind Event to On Input Method Changed(self=@n7, Delegate=Create Event())
   > OnInputMethodChanged!(execute, self, bNewInputType:ECommonInputType Enum=GetCurrentInputType(self=@n7)) -> (then)
   Event On Deactivated()
   > Unbind Event from On Input Method Changed(self=Get CommonInputSubsystem(), Delegate=Create Event())
  ]}
 HandleIslandChanged {nodes [
   HandleIslandChanged!() -> (then, Island:Online Link Id Structure) {is=title K2Node_FunctionEntry}
   > Branch(Condition=IsValid(Object=Get SelectedIslandViewModel()))
   n1 = Get SelectedIslandViewModel()
   > "SetText (Text)"(
    self=Get Txt_PlayingNow() -> (Txt_PlayingNow:Common Text Object Reference)
    InText=Format Text!(
     Format="LOCTEXT(\"Playing {Game Mode}\")"
     Game Mode:Text=Island Title!(self:Selected Island View Model Object Reference=@n1) -> (ReturnValue:Text)
    ) -> (Result)
   )
   n2 = GetCCU!(
    self:Selected Island View Model Object Reference=Get SelectedIslandViewModel()
    LocalPlayer:Local Player Object Reference=GetOwningLocalPlayer!(self:Widget Object Reference) -> (ReturnValue:Local Player Object Reference)
   ) -> (ReturnValue:int=0)
   > "SetText (Text)"(
    self=Get Text_PlayersAmount() -> (Text_PlayersAmount:Common Text Object Reference)
    InText=SelectText(
     A=GetShorthandQuantity!(Quantity:int=@n2, MinimumFractionalDigits:int=0, MaximumFractionalDigits:int=1, bShouldRoundDown:bool=false) -> (ReturnValue:Text)
     B="To Text (Integer)"(Value=@n2)
     bPickA="integer > integer"(A=@n2, B=999)
    )
   )
   > SetVisibility(
    self=Get HB_Players() -> (HB_Players:Horizontal Box Object Reference)
    InVisibility=Select(Option 0=SelfHitTestInvisible, Option 1=Collapsed, Index="integer <= integer"(A=@n2))
   )
  ]}
 ListDownNavigation {nodes [
   ListDownNavigation!() -> (then, Navigation:EUINavigation Enum) {is=title K2Node_FunctionEntry}
   n1 = Get Button_Matchmaking() -> (Button_Matchmaking:WBP FN1 UI Button Matchmaking Object Reference)
   > Return Node(ReturnValue=Select(
     Option 0:Widget Object Reference=Get Button_Close()
     Option 1:Widget Object Reference=@n1
     Index=GetIsEnabled!(self:Widget Object Reference=@n1) -> (ReturnValue:bool)
    ) -> (ReturnValue:Widget Object Reference))
  ]}
 NavigateLeftFromPlayButton {nodes [
   NavigateLeftFromPlayButton!() -> (then, Navigation:EUINavigation Enum) {is=title K2Node_FunctionEntry}
   n1 = Get PolyList()
   n2 = Get ListView(self=@n1)
   > Return Node(ReturnValue=GetWidgetFromItem!(
     self:Fort Polymorphic View Object Reference=@n1
     Item:Object Reference=Get Selected Item!(self:List View Object Reference=@n2) -> (ReturnValue:Object Reference) {is=title K2Node_CallFunction}
    ) -> (ReturnValue:User Widget Object Reference))
  ]}
 OnInputMethodChanged {nodes [
   n1 = OnInputMethodChanged!() -> (then, bNewInputType:ECommonInputType Enum) {is=title K2Node_FunctionEntry}
   > n2 = Switch on ECommonInputType!(execute, Selection:ECommonInputType Enum=@n1) -> (MouseAndKeyboard:Exec, Gamepad:Exec, Touch:Exec, Count:Exec) {is=title K2Node_SwitchEnum}
   RequestRefreshFocus(execute=@n2.Gamepad, self:Common Activatable Widget Object Reference)
  ]}
 PopulateSettings list {nodes [
   n1 = PopulateSettings list!() -> (then, InListItems:Array of Object References) {is=title K2Node_FunctionEntry}
   > SetListItems!(execute, self:Fort Polymorphic View Object Reference=Get PolyList(), InListItems:Array of Object References=@n1) -> (then)
   > RequestRefreshFocus()
   > ResetInitialAnalyticsAttributes(self:WBP FN1 Matchmaking Settings Poly List Object Reference=Get PolyList())
  ]}
 SetBackgroundImageURL {nodes [
   n1 = SetBackgroundImageURL!() -> (then, URL:String) {is=title K2Node_FunctionEntry}
   > SetMediaURL!(execute, self:Epic CMSImage Object Reference=Get BackgroundCMS() -> (BackgroundCMS:Epic CMSImage Object Reference), MediaUrl:String=@n1) -> (then)
  ]}
 SetIsRanked {nodes [
   n1 = SetIsRanked!() -> (then, IsRanked:bool=~) {is=title K2Node_FunctionEntry}
   > n2 = Branch(Condition=Get DontPlayRankTransition() -> (DontPlayRankTransition:bool))
   > Set DontPlayRankTransition(DontPlayRankTransition=false)
   n3 = "To Float (Boolean)"(InBool=@n1)
   > SetScalarParameterValue!(
    execute
    Collection:Material Parameter Collection Object Reference=Get MaterialCollection() -> (MaterialCollection:Material Parameter Collection Object Reference)
    ParameterName:Name=MakeLiteralName(Value=Ranked)
    ParameterValue:float=@n3
   ) -> (then) {is=title K2Node_CallMaterialParameterCollectionFunction}
   > Set RenderOpacity!(execute, RenderOpacity:float=@n3, self:Widget Object Reference=Get RankedPattern() -> (RankedPattern:Image Object Reference)) -> (then, Output_Get:float)
   > SetRanked!(execute, self:WBP FN1 UI Button Matchmaking Object Reference=Get Button_Matchmaking(), Ranked:double=@n3) -> (then)
   n4 = Branch(execute=@n2.else, Condition=@n1)
   n5 = Get Anim_SwitchRank() -> (Anim_SwitchRank:Widget Animation Object Reference)
   > PlayAnimationForward(InAnimation=@n5)
   PlayAnimationReverse(execute=@n4.else, InAnimation=@n5)
  ]}
 SetLogoURL {nodes [
   n1 = SetLogoURL!() -> (then, URL:String) {is=title K2Node_FunctionEntry}
   n2 = IsEmpty(InString=@n1)
   > SetActiveWidget(
    self=Get Switcher_Logo() -> (Switcher_Logo:Common Visibility Switcher Object Reference)
    Widget=Select(
     Option 0=Get LogoCMS() -> (LogoCMS:Epic CMSImage Object Reference)
     Option 1=Get Txt_LogoFallback() -> (Txt_LogoFallback:Common Text Object Reference)
     Index=@n2
    )
   )
   > n3 = Branch(Condition=@n2)
   SetMediaURL(execute=@n3.else, self=Get LogoCMS(), MediaUrl=@n1)
  ]}
 SetMobileStyle {nodes [
   SetMobileStyle() {is=title K2Node_FunctionEntry}
   > Branch(Condition=IsUsingCompactUI())
   n1 = CompareToScreenAspectRatio!(Comparison:EAspectRatioComparison Enum=LessThanOrEqualTo, AspectRatio:EAspectRatio Enum=Ratio_4_3) -> (ReturnValue:bool)
   > SetPadding!(
    execute
    self:Grid Slot Object Reference=SlotAsGridSlot(Widget=Get O_ListClipping() -> (O_ListClipping:Overlay Object Reference))
    InPadding_Left:float=-50
    InPadding_Top:float=SelectFloat(A=64, B=0, bPickA=@n1)
    InPadding_Right:float=-50
    InPadding_Bottom:float=24
   ) -> (then)
   > Branch(Condition=@n1)
   > SetHorizontalAlignment(
    self:Horizontal Box Slot Object Reference=SlotAsHorizontalBoxSlot!(Widget:Widget Object Reference=Get SZ_Content_R() -> (SZ_Content_R:Safe Zone Object Reference)) -> (ReturnValue:Horizontal Box Slot Object Reference)
    InHorizontalAlignment=HAlign_Fill
   )
  ]}
 SetRankTransition {nodes [
   n1 = SetRankTransition!() -> (then, RankTransition:double=~) {is=title K2Node_FunctionEntry}
   > SetScalarParameterValue(Collection=Get MaterialCollection(), ParameterName=MakeLiteralName(Value=Ranked), ParameterValue=@n1)
   > SetRanked(self=Get Button_Matchmaking(), Ranked=@n1)
  ]}
]
variables [
 DontPlayRankTransition {type=boolean}
 IsPlayClicked {type=boolean}
 MaterialCollection {type=MaterialParameterCollection, default=/MatchmakingSettingsFN1Shooters/MatchmakingSettings/Materials/MPC_FN1_Matchmaking_Ranked}
 RankTransition {type=number}
]
widgets {
 info {
  inheritedWidgetCount=0
  parentClass=/Script/MatchmakingSettingsUI.MatchmakingSettingsWidget
  rootWidgetClass=&GridPanel
  widgetCount=28
 }
 widgets [
  Root {widgetClassPath=&GridPanel, bIsVariable=true, widget=Root, properties {bOverride_Cursor=true, columnFill [1,1], rowFill [1]}}
  BackgroundPlaceholder {
   widgetClassPath=&Image
   parent=Root
   widget=BackgroundPlaceholder
   properties {colorAndOpacity=<rgba 0 0 0 0.5>}
   slot {class=GridSlot, properties {columnSpan=2, layer=-100}}
  }
  Background_Options {
   widgetClassPath=&Image
   bIsVariable=true
   parent=Root
   widget=Background_Options
   properties {brush {
     bIsDynamicallyLoaded=false
     drawAs=Image
     imageType=FullColor
     mirroring=NoMirror
     resourceName=None
     resourceObject=/MatchmakingSettingsFN1Shooters/MatchmakingSettings/MaterialInstances/MI_UI_FN1_SettingsBackground
     tiling=NoTile
     imageSize=<0 0>
     margin=<ltrb 0 0 0 0>
     outlineSettings {
      bUseBrushTransparency=false
      roundingType=HalfHeightRadius
      width=0
      color {colorUseRule=UseColor_Specified, specifiedColor=<rgba 0 0 0 0>}
      cornerRadii=<0 0 0 0>
     }
     tintColor {colorUseRule=UseColor_Specified, specifiedColor=<rgba 1 1 1 1>}
     uVRegion {bIsValid=false, max=<0 0>, min=<0 0>}
    }}
   slot {class=GridSlot, properties {column=1, layer=-6, rowSpan=2}}
  }
  BackgroundCMS {
   widgetClassPath=/Script/EpicCMSUIFramework.EpicCMSImage
   bIsVariable=true
   parent=Root
   widget=BackgroundCMS
   properties {
    materialTextureParamName=Texture
    brush {
     bIsDynamicallyLoaded=false
     drawAs=Image
     imageType=FullColor
     mirroring=NoMirror
     resourceName=None
     resourceObject=/MatchmakingSettingsFN1Shooters/MatchmakingSettings/Materials/MI_UI_FN1_GameArtBackground
     tiling=NoTile
     imageSize=<32 32>
     margin=<ltrb 0 0 0 0>
     outlineSettings {
      bUseBrushTransparency=false
      roundingType=HalfHeightRadius
      width=0
      color {colorUseRule=UseColor_Specified, specifiedColor=<rgba 0 0 0 0>}
      cornerRadii=<0 0 0 0>
     }
     tintColor {colorUseRule=UseColor_Specified, specifiedColor=<rgba 1 1 1 1>}
     uVRegion {bIsValid=false, max=<0 0>, min=<0 0>}
    }
   }
   slot {class=GridSlot, properties {layer=-10, padding=<ltrb 0 0 -320 0>}}
  }
  WBP_CaptureForPostBufferUpdate {
   widgetClassPath=&WBP_CaptureForPostBufferUpdate
   parent=Root
   slot {class=GridSlot, properties {columnSpan=2, layer=-5, rowSpan=2}}
  }
  SZ_Content_R {
   widgetClassPath=&SafeZone
   bIsVariable=true
   parent=Root
   widget=SZ_Content_R
   properties {padLeft=false}
   slot {class=GridSlot, properties {column=1, horizontalAlignment=HAlign_Left, layer=20, padding=<ltrb 233 0 0 0>}}
  }
  SizeBox_Content {
   widgetClassPath=&SizeBox
   parent=SZ_Content_R
   widget=SizeBox_Content
   properties {bOverride_WidthOverride=true, widthOverride=592}
   slot {class=SafeZoneSlot, properties {hAlign=HAlign_Left, padding=<ltrb 0 80 84 0>}}
  }
  VB_Content {
   widgetClassPath=&GridPanel
   parent=SizeBox_Content
   widget=VB_Content
   properties {bOverride_Cursor=true, columnFill [1], rowFill [1,0,0]}
   slot {class=SizeBoxSlot}
  }
  O_ListClipping {
   widgetClassPath=&Overlay
   bIsVariable=true
   parent=VB_Content
   widget=O_ListClipping
   properties {clipping=ClipToBounds}
   slot {class=GridSlot, properties {layer=10, padding=<ltrb -50 0 -50 24>}}
  }
  PolyList {
   widgetClassPath=&/WBP_FN1_MatchmakingSettings_PolyList
   bIsVariable=true
   parent=O_ListClipping
   widget=PolyList
   properties {
    bIsFocusable=true
    navigation=&self:WidgetTree.PolyList.WidgetNavigation_0
   }
   slot {class=OverlaySlot, properties {horizontalAlignment=HAlign_Fill, verticalAlignment=VAlign_Fill, padding=<ltrb 50 16 20 16>}}
  }
  WBP_FN1_MatchmakingSettings_ErrorState {
   widgetClassPath=&/WBP_FN1_MatchmakingSettings_ErrorState
   bIsVariable=true
   parent=VB_Content
   widget=WBP_FN1_MatchmakingSettings_ErrorState
   slot {class=GridSlot, properties {row=1, padding=<ltrb 2 0 0 16>}}
  }
  HorizontalBox_0 {
   widgetClassPath=&HorizontalBox
   parent=VB_Content
   widget=HorizontalBox_0
   properties {navigation=&self:WidgetTree.HorizontalBox_0.WidgetNavigation_1}
   slot {class=GridSlot, properties {row=2, padding=<ltrb 0 0 0 80>}}
  }
  Button_Matchmaking {
   widgetClassPath=&/WBP_FN1_UI_Button_Matchmaking
   bIsVariable=true
   parent=HorizontalBox_0
   widget=Button_Matchmaking
   properties {
    bDisplayInputActionWhenNotInteractable=true
    bHideInputAction=false
    bHideInputActionWithKeyboard=true
    bShouldUseFallbackDefaultInputAction=false
    maxWidth=348
    minHeight=56
    minHeightOverride=80
    minWidthOverride=348
    mobileMinHeight=80
    shouldPromptDirectionFollowCulture=false
    useDarkInputIcon=true
    buttonPadding=<ltrb 40 0 40 0>
    font {
     bForceMonospaced=false
     bMaterialIsStencil=false
     fontMaterial=/MatchmakingSettingsFN1Shooters/MatchmakingSettings/MaterialInstances/MI_MM_PlayColorText
     fontObject=&Burbank
     letterSpacing=10
     monospacedWidth=1
     size=30
     skewAmount=0
     typefaceFontName=BurbankRegular-Black
     outlineSettings {
      bApplyOutlineToDropShadows=false
      bMiteredCorners=false
      bSeparateFillAlpha=false
      outlineMaterial=null
      outlineSize=0
      outlineColor=<rgba 0 0 0 1>
     }
    }
    triggeringInputAction {dataTable=/MatchmakingSettingsFN1Shooters/MatchmakingSettings/Data/FN1ShootersInputActionDataTable, rowName=Play}
   }
   slot {class=HorizontalBoxSlot, properties {size {sizeRule=Fill, value=1}}}
  }
  Button_Close {
   widgetClassPath=&WBP_UIKit_Button_Generic
   bIsVariable=true
   parent=HorizontalBox_0
   widget=Button_Close
   properties {
    bDisplayInputActionWhenNotInteractable=true
    bShouldUseFallbackDefaultInputAction=false
    backgroundClass=&/WBP_MatchmakingSettingsBackground.WBP_MatchmakingSettingsBackground_C
    backgroundMaterial=/MatchmakingSettingsFN1Shooters/MatchmakingSettings/MaterialInstances/MI_UI_MMSettingsCloseBackground
    maxWidth=228
    minHeightOverride=80
    minWidth=288
    minWidthOverride=228
    mobileFontSizeMultiplierOverride=1.4
    mobileMinHeight=80
    mobileMinWidth=288
    overrideDefaultMobileFontSizeMultiplier=true
    text=Done
    useDarkInputIcon=true
    buttonPadding=<ltrb 32 0 32 0>
    font {
     bForceMonospaced=false
     bMaterialIsStencil=false
     fontMaterial=/FortUILibrary/UIKit/Buttons/MaterialInstances/CTAButtons/MI_UIKit_TextColor_Quiet
     fontObject=&HeadingNow
     letterSpacing=48
     monospacedWidth=1
     size=15
     skewAmount=0
     typefaceFontName=86 Bold
     outlineSettings {
      bApplyOutlineToDropShadows=false
      bMiteredCorners=false
      bSeparateFillAlpha=false
      outlineMaterial=null
      outlineSize=0
      outlineColor=<rgba 0 0 0 1>
     }
    }
    promptNudge=<8 0>
    secondFont {
     bForceMonospaced=false
     bMaterialIsStencil=false
     fontMaterial=/MatchmakingSettingsFN1Shooters/MatchmakingSettings/MaterialInstances/MI_MM_CloseColorText
     fontObject=&Burbank
     letterSpacing=48
     monospacedWidth=1
     size=21
     skewAmount=0
     typefaceFontName=BurbankRegular-Black
     outlineSettings {
      bApplyOutlineToDropShadows=false
      bMiteredCorners=false
      bSeparateFillAlpha=false
      outlineMaterial=null
      outlineSize=0
      outlineColor=<rgba 0 0 0 1>
     }
    }
    triggeringInputAction {dataTable=&InputActionDataTable, rowName=Back}
   }
   slot {class=HorizontalBoxSlot, properties {verticalAlignment=VAlign_Bottom, padding=<ltrb 16 0 0 0>}}
  }
  RankedPattern {
   widgetClassPath=&Image
   bIsVariable=true
   parent=Root
   widget=RankedPattern
   properties {
    renderOpacity=0
    visibility=SelfHitTestInvisible
    brush {
     bIsDynamicallyLoaded=false
     drawAs=Image
     imageType=FullColor
     mirroring=NoMirror
     resourceName=None
     resourceObject=/MatchmakingSettingsFN1Shooters/MatchmakingSettings/Materials/M_UI_FN1_RankedPattern
     tiling=NoTile
     imageSize=<40 32>
     margin=<ltrb 0 0 0 0>
     outlineSettings {
      bUseBrushTransparency=false
      roundingType=HalfHeightRadius
      width=0
      color {colorUseRule=UseColor_Specified, specifiedColor=<rgba 0 0 0 0>}
      cornerRadii=<0 0 0 0>
     }
     tintColor {colorUseRule=UseColor_Specified, specifiedColor=<rgba 1 1 1 1>}
     uVRegion {bIsValid=false, max=<0 0>, min=<0 0>}
    }
    colorAndOpacity=<rgba 0.5776 1 0 1>
   }
   slot {class=GridSlot, properties {column=1, horizontalAlignment=HAlign_Right, layer=5}}
  }
  SZ_MapInfo {
   widgetClassPath=&SafeZone
   parent=Root
   widget=SZ_MapInfo
   properties {padRight=false}
   slot {class=GridSlot, properties {horizontalAlignment=HAlign_Right, padding=<ltrb 0 0 -320 0>}}
  }
  SizeBox_1 {
   widgetClassPath=&SizeBox
   parent=SZ_MapInfo
   widget=SizeBox_1
   properties {bOverride_WidthOverride=true, maxDesiredWidth=1280, widthOverride=1280}
   slot {class=SafeZoneSlot}
  }
  GridPanel_69 {
   widgetClassPath=&GridPanel
   parent=SizeBox_1
   widget=GridPanel_69
   properties {columnFill [1], rowFill [0,1]}
   slot {class=SizeBoxSlot}
  }
  StackBox_0 {
   widgetClassPath=&StackBox
   parent=GridPanel_69
   widget=StackBox_0
   properties {orientation=Orient_Vertical}
   slot {class=GridSlot, properties {horizontalAlignment=HAlign_Left, row=1, verticalAlignment=VAlign_Bottom, padding=<ltrb 104 0 0 80>}}
  }
  Switcher_Logo {
   widgetClassPath=&CommonVisibilitySwitcher
   bIsVariable=true
   parent=StackBox_0
   widget=Switcher_Logo
   slot {class=StackBoxSlot, properties {horizontalAlignment=HAlign_Left, verticalAlignment=VAlign_Top}}
  }
  Txt_LogoFallback {
   widgetClassPath=&CommonTextBlock
   bIsVariable=true
   parent=Switcher_Logo
   widget=Txt_LogoFallback
   properties {
    text=Title Fallback
    wrapTextAt=696
    font {
     bForceMonospaced=false
     bMaterialIsStencil=false
     fontMaterial=null
     fontObject=&NotoSans
     letterSpacing=0
     monospacedWidth=1
     size=24
     skewAmount=0
     typefaceFontName=NotoSans-Regular
     outlineSettings {
      bApplyOutlineToDropShadows=false
      bMiteredCorners=false
      bSeparateFillAlpha=false
      outlineMaterial=null
      outlineSize=0
      outlineColor=<rgba 0 0 0 1>
     }
    }
    shadowOffset=<0 0>
   }
   slot {class=CommonVisibilitySwitcherSlot, properties {horizontalAlignment=HAlign_Left, verticalAlignment=VAlign_Top}}
  }
  LogoCMS {
   widgetClassPath=&EpicCMSImage
   bIsVariable=true
   parent=Switcher_Logo
   widget=LogoCMS
   properties {brush {
     bIsDynamicallyLoaded=false
     drawAs=Image
     imageType=NoImage
     mirroring=NoMirror
     resourceName=None
     resourceObject=null
     tiling=NoTile
     imageSize=<480 270>
     margin=<ltrb 0 0 0 0>
     outlineSettings {
      bUseBrushTransparency=false
      roundingType=HalfHeightRadius
      width=0
      color {colorUseRule=UseColor_Specified, specifiedColor=<rgba 0 0 0 0>}
      cornerRadii=<0 0 0 0>
     }
     tintColor {colorUseRule=UseColor_Specified, specifiedColor=<rgba 1 1 1 1>}
     uVRegion {bIsValid=false, max=<0 0>, min=<0 0>}
    }}
   slot {class=CommonVisibilitySwitcherSlot, properties {horizontalAlignment=HAlign_Left}}
  }
  MapRotation_Entry {
   widgetClassPath=/MatchmakingSettingsFN1Shooters/MatchmakingSettings/Entries/MapRotation/WBP_FN1_MatchmakingSettings_MapRotation_Entry
   bIsVariable=true
   parent=StackBox_0
   widget=MapRotation_Entry
   slot {class=StackBoxSlot, properties {padding=<ltrb 0 48 0 0>}}
  }
  HB_Players {
   widgetClassPath=&HorizontalBox
   bIsVariable=true
   parent=GridPanel_69
   widget=HB_Players
   slot {class=GridSlot, properties {horizontalAlignment=HAlign_Left, verticalAlignment=VAlign_Top, padding=<ltrb 104 58 0 0>}}
  }
  Border_41 {
   widgetClassPath=&Border
   parent=HB_Players
   widget=Border_41
   properties {brushColor=<rgba 0.0048 0.004 0.0116 1>, padding=<ltrb 12 1 12 1>}
   slot {class=HorizontalBoxSlot, properties {verticalAlignment=VAlign_Top}}
  }
  Text_PlayersAmount {
   widgetClassPath=&CommonTextBlock
   bIsVariable=true
   parent=Border_41
   widget=Text_PlayersAmount
   properties {
    scrollStyle=&TextScrollStyle-Base_C
    text=800K
    textTransformPolicy=ToUpper
    colorAndOpacity {colorUseRule=UseColor_Specified, specifiedColor=<rgba 0.8879 0.8879 0.8879 1>}
    font {
     bForceMonospaced=false
     bMaterialIsStencil=false
     fontMaterial=null
     fontObject=&BurbankSmall
     letterSpacing=0
     monospacedWidth=1
     size=18
     skewAmount=0
     typefaceFontName=Black
     outlineSettings {
      bApplyOutlineToDropShadows=false
      bMiteredCorners=false
      bSeparateFillAlpha=false
      outlineMaterial=null
      outlineSize=0
      outlineColor=<rgba 0 0 0 1>
     }
    }
    shadowOffset=<0 0>
   }
   slot {class=BorderSlot, properties {padding=<ltrb 12 1 12 1>}}
  }
  Txt_PlayingNow {
   widgetClassPath=&CommonTextBlock
   bIsVariable=true
   parent=HB_Players
   widget=Txt_PlayingNow
   properties {
    text=Playing Now
    textTransformPolicy=ToUpper
    wrapTextAt=900
    colorAndOpacity {colorUseRule=UseColor_Specified, specifiedColor=<rgba 0.8879 0.8879 0.8879 1>}
    font {
     bForceMonospaced=false
     bMaterialIsStencil=false
     fontMaterial=null
     fontObject=&BurbankSmall
     letterSpacing=0
     monospacedWidth=1
     size=18
     skewAmount=0
     typefaceFontName=Black
     outlineSettings {
      bApplyOutlineToDropShadows=false
      bMiteredCorners=false
      bSeparateFillAlpha=false
      outlineMaterial=null
      outlineSize=2
      outlineColor=<rgba 0.0048 0.004 0.0116 1>
     }
    }
    shadowOffset=<0 0>
   }
   slot {class=HorizontalBoxSlot, properties {horizontalAlignment=HAlign_Left, verticalAlignment=VAlign_Center, padding=<ltrb 18 0 0 0>}}
  }
  Vignettes {
   widgetClassPath=&Image
   parent=Root
   widget=Vignettes
   properties {
    visibility=SelfHitTestInvisible
    brush {
     bIsDynamicallyLoaded=false
     drawAs=Image
     imageType=FullColor
     mirroring=NoMirror
     resourceName=None
     resourceObject=/MatchmakingSettingsFN1Shooters/MatchmakingSettings/Materials/M_UI_FN1_Vignettes
     tiling=NoTile
     imageSize=<32 32>
     margin=<ltrb 0 0 0 0>
     outlineSettings {
      bUseBrushTransparency=false
      roundingType=HalfHeightRadius
      width=0
      color {colorUseRule=UseColor_Specified, specifiedColor=<rgba 0 0 0 0>}
      cornerRadii=<0 0 0 0>
     }
     tintColor {colorUseRule=UseColor_Specified, specifiedColor=<rgba 1 1 1 1>}
     uVRegion {bIsValid=false, max=<0 0>, min=<0 0>}
    }
   }
   slot {class=GridSlot, properties {columnSpan=2}}
  }
 ]
}
