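# WBP_FN1_Matchmaking_EntryDescription is a CommonUserWidget that shows description text, apparently for an entry in the FN1 Shooters matchmaking settings UI.
# Its only widget is a styled CommonTextBlock named Description, using the HeadingNow font, a mobile size multiplier and a scroll style.
# A SetText function passes its input text straight to that block; the event graph events have no logic attached.
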WidgetBlueprint /MatchmakingSettingsFN1Shooters/MatchmakingSettings/WBP_FN1_Matchmaking_EntryDescription
graphs [
 EventGraph {nodes [
   Event PreConstruct()
   Event Construct()
   Event Tick()
  ]}
 SetText {nodes [
   n1 = SetText!() -> (then, InText:Text) {is=title K2Node_FunctionEntry}
   > "SetText (Text)"(self=Get Description() -> (Description:Common Text Object Reference), InText=@n1)
  ]}
]
variables []
widgets {
 info {inheritedWidgetCount=0, parentClass=&CommonUserWidget, rootWidgetClass=&CommonTextBlock, widgetCount=1}
 widgets [Description {
   widgetClassPath=&CommonTextBlock
   bIsVariable=true
   widget=Description
   properties {
    mobileFontSizeMultiplier=1.2
    scrollStyle=&TextScrollStyle-Base_C
    font {
     bForceMonospaced=false
     bMaterialIsStencil=false
     fontMaterial=null
     fontObject=&HeadingNow
     letterSpacing=0
     monospacedWidth=1
     size=12.75
     skewAmount=0
     typefaceFontName=75 Medium
     outlineSettings {
      bApplyOutlineToDropShadows=false
      bMiteredCorners=false
      bSeparateFillAlpha=false
      outlineMaterial=null
      outlineSize=0
      outlineColor=<rgba 0 0 0 1>
     }
    }
    shadowOffset=<0 0>
   }
  }]
}
