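# Matchmaking settings outline widget: a single non-hit-testable Image, hidden by default, that draws a focus outline using a settings UI material.
# PreConstruct rounds the corners through a dynamic material parameter.
# ShowOutline sets opacity to full only in compact UI with touch input and a caller-supplied condition; HideOutline clears it in compact UI.
# Both switch opacity instantly; the short AnimShowOutline fade exists but these graphs never play it.
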WidgetBlueprint /MatchmakingSettingsFN1Shooters/MatchmakingSettings/WBP_Matchmaking_Settings_Outline
animations [AnimShowOutline {
  binding_count=1
  length_seconds=0.1
  object=&self:AnimShowOutline
  sequence {bindings [FocusOutline {
     id=#1
     possessed_class=Image
     tracks [{
       type=MovieSceneFloatTrack
       display_name=Render Opacity
       sections [{type=MovieSceneFloatSection, end_frame=2, start_frame=0, channels [None {type=MovieSceneScriptingFloatChannel, keys [[0,0],[2,1]]}]}]
      }]
    }]}
 }]
graphs [
 EventGraph {nodes [
   n1 = Event PreConstruct()
   > Parent: PreConstruct(IsDesignTime=@n1.IsDesignTime)
   > n2 = GetDynamicMaterial(self=Get FocusOutline() -> (FocusOutline:Image Object Reference))
   > SetVectorParameterValue!(execute, self=@n2, ParameterName=Corner Radius, Value_R:float=0.15, Value_G:float=0.15, Value_B:float=0.15, Value_A:float=0.15) -> (then)
  ]}
 HideOutline {nodes [
   HideOutline() {is=title K2Node_FunctionEntry}
   n1 = Get RenderOpacity!(self:Widget Object Reference=Get FocusOutline()) -> (RenderOpacity:float)
   > Branch(Condition=AND Boolean(A=IsUsingCompactUI(), B="float > float"(A=@n1, B=0.5)))
   > SetRenderOpacity(self=Get FocusOutline(), InOpacity=0)
  ]}
 ShowOutline {nodes [
   n1 = ShowOutline!() -> (then, AdditionalCondition:bool=true) {is=title K2Node_FunctionEntry}
   n2 = GetCurrentInputType(self=Get CommonInputSubsystem())
   n3 = Get RenderOpacity(self=Get FocusOutline())
   > Branch(Condition=AND Boolean!(A=@n1, B=IsUsingCompactUI(), C:bool="float < float"(A=@n3, B=0.5), D:bool="Equal (Enum)"(A=@n2, B=Touch)) -> (ReturnValue))
   > SetRenderOpacity(self=Get FocusOutline(), InOpacity=1)
  ]}
]
variables []
widgets {
 info {inheritedWidgetCount=0, parentClass=&UserWidget, rootWidgetClass=&Image, widgetCount=1}
 widgets [FocusOutline {
   widgetClassPath=&Image
   bIsVariable=true
   widget=FocusOutline
   properties {
    pixelSnapping=Disabled
    renderOpacity=0
    visibility=SelfHitTestInvisible
    brush {
     bIsDynamicallyLoaded=false
     drawAs=Image
     imageType=FullColor
     mirroring=NoMirror
     resourceName=None
     resourceObject=/FortniteCoreUI/Settings/Art/MI_UIKit_Settings_Outline
     tiling=NoTile
     imageSize=<0 0>
     margin=<ltrb 0 0 0 0>
     outlineSettings {
      bUseBrushTransparency=false
      roundingType=HalfHeightRadius
      width=0
      color {colorUseRule=UseColor_Specified, specifiedColor=<rgba 0 0 0 0>}
      cornerRadii=<0 0 0 0>
     }
     tintColor {colorUseRule=UseColor_Specified, specifiedColor=<rgba 1 1 1 1>}
     uVRegion {bIsValid=false, max=<0 0>, min=<0 0>}
    }
   }
  }]
}
